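# Animation Blueprint for the DarkAzalea female body that adds corrective deformation and knife visibility over the parent mesh's copied pose.
# Each update derives alphas from joint rotation deltas, driving constraints and additive bone offsets at thighs, belt, shoulders, elbows and knees; neck nodes are disconnected.
# Bone scaling shows foot knives while airborne and hand knives according to skydiving events and an equipped DarkAzalea-tagged pickaxe.
# Customization load zeroes two material hide parameters.
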AnimBlueprint /BRCosmetics/Characters/Player/Female/Medium/Bodies/F_MED_DarkAzalea/Meshes/F_MED_DarkAzalea_AnimBP1
target_skeleton=&/F_MED_DarkAzalea_Skeleton
graphs [
 AnimGraph {nodes [
   n1 = Get neck_rot_up() -> (neck_rot_up:double)
   n2 = "Constraint - neck_NecklineShape (PP)"!(ComponentPose, Alpha=@n1, ConstraintWeights_0=0.7, ConstraintWeights_1:float=0.3) -> (Pose) {
    alphaBoolBlend {blendOption=Linear}
    boneToModify {boneName=neck_NecklineShape}
    constraintSetup [{targetBone {boneName=neck_02}}, {targetBone {boneName=spine_05}}]
    constraintWeights [1,1]
   }
   "Transform (Modify) Bone - Bone: neck_NecklineShape"(ComponentPose=@n2, Alpha=@n1, Translation="0,1,0", Rotation="0,0,0", Scale="1,1,1") {
    rotationMode=BMM_Additive
    rotationSpace=BCS_BoneSpace
    scaleMode=BMM_Additive
    scaleSpace=BCS_BoneSpace
    translationMode=BMM_Additive
    translationSpace=BCS_BoneSpace
    alphaBoolBlend {blendOption=Linear}
    boneToModify {boneName=neck_NecklineShape}
    translation {y=1}
   }
   n3 = Get neck_rot_dn() -> (neck_rot_dn:double)
   n4 = "Transform (Modify) Bone - Bone: neck_NecklineShape"(Alpha=@n3, Translation="-1,1,0", Rotation="0,0,0", Scale="1,1,1") {
    rotationMode=BMM_Additive
    rotationSpace=BCS_BoneSpace
    scaleMode=BMM_Additive
    scaleSpace=BCS_BoneSpace
    translationMode=BMM_Additive
    translationSpace=BCS_BoneSpace
    alphaBoolBlend {blendOption=Linear}
    boneToModify {boneName=neck_NecklineShape}
    translation=<-1 1>
   }
   "Transform (Modify) Bone - Bone: neck_zipper"(ComponentPose=@n4, Alpha=@n3, Translation="0,0,0", Rotation="0,-90,0", Scale="1,1,1") {
    rotationMode=BMM_Additive
    rotationSpace=BCS_BoneSpace
    scaleMode=BMM_Additive
    scaleSpace=BCS_BoneSpace
    translationMode=BMM_Additive
    translationSpace=BCS_BoneSpace
    alphaBoolBlend {blendOption=Linear}
    boneToModify {boneName=neck_zipper}
    rotation {yaw=-90}
   }
   n5 = "Constraint - thigh_volume_bk_l (PP)"!(
    ComponentPose=Local To Component(LocalPose=Copy Pose From Mesh() {bUseAttachedParent=true})
    Alpha=1
    ConstraintWeights_0=0.4
    ConstraintWeights_1:float=0.6
   ) -> (Pose) {
    alphaBoolBlend {blendOption=Linear}
    boneToModify {boneName=thigh_volume_bk_l}
    constraintSetup [{targetBone {boneName=thigh_twist_01_l}}, {targetBone {boneName=pelvis}}]
    constraintWeights [1,1]
   }
   n6 = "Constraint - thigh_volume_bk_r (PP)"!(ComponentPose=@n5, Alpha=1, ConstraintWeights_0=0.4, ConstraintWeights_1:float=0.6) -> (Pose) {
    alphaBoolBlend {blendOption=Linear}
    boneToModify {boneName=thigh_volume_bk_r}
    constraintSetup [{targetBone {boneName=thigh_twist_01_r}}, {targetBone {boneName=pelvis}}]
    constraintWeights [1,1]
   }
   n7 = Get thigh_shape_fn_l() -> (thigh_shape_fn_l:double)
   n8 = "Constraint - thigh_volume_bk_l (PP)"(ComponentPose=@n6, Alpha=@n7, ConstraintWeights_0=0, ConstraintWeights_1=1) {
    alphaBoolBlend {blendOption=Linear}
    boneToModify {boneName=thigh_volume_bk_l}
    constraintSetup [{targetBone {boneName=thigh_twist_01_l}}, {targetBone {boneName=pelvis}}]
    constraintWeights [1,1]
   }
   n9 = Get thigh_shape_fn_r() -> (thigh_shape_fn_r:double)
   n10 = "Constraint - thigh_volume_bk_r (PP)"(ComponentPose=@n8, Alpha=@n9, ConstraintWeights_0=0, ConstraintWeights_1=1) {
    alphaBoolBlend {blendOption=Linear}
    boneToModify {boneName=thigh_volume_bk_r}
    constraintSetup [{targetBone {boneName=thigh_twist_01_r}}, {targetBone {boneName=pelvis}}]
    constraintWeights [1,1]
   }
   n11 = "Constraint - thigh_volume_fn_l (PP)"!(ComponentPose=@n10, Alpha=@n7, ConstraintWeights_0=0.5, ConstraintWeights_1:float=0.5) -> (Pose) {
    alphaBoolBlend {blendOption=Linear}
    boneToModify {boneName=thigh_volume_fn_l}
    constraintSetup [{targetBone {boneName=thigh_twist_01_l}}, {targetBone {boneName=pelvis}}]
    constraintWeights [1,1]
   }
   n12 = "Constraint - thigh_volume_fn_r (PP)"!(ComponentPose=@n11, Alpha=@n9, ConstraintWeights_0=0.5, ConstraintWeights_1:float=0.5) -> (Pose) {
    alphaBoolBlend {blendOption=Linear}
    boneToModify {boneName=thigh_volume_fn_r}
    constraintSetup [{targetBone {boneName=thigh_twist_01_r}}, {targetBone {boneName=pelvis}}]
    constraintWeights [1,1]
   }
   n13 = "Constraint - thigh_volume_lt_l1 (PP)"!(ComponentPose=@n12, Alpha=1, ConstraintWeights_0=0.5, ConstraintWeights_1:float=0.5) -> (Pose) {
    alphaBoolBlend {blendOption=Linear}
    boneToModify {boneName=thigh_volume_lt_l1}
    constraintSetup [{targetBone {boneName=thigh_twist_01_l}}, {targetBone {boneName=pelvis}}]
    constraintWeights [1,1]
   }
   n14 = "Constraint - thigh_volume_rt_r (PP)"!(ComponentPose=@n13, Alpha=1, ConstraintWeights_0=0.5, ConstraintWeights_1:float=0.5) -> (Pose) {
    alphaBoolBlend {blendOption=Linear}
    boneToModify {boneName=thigh_volume_rt_r}
    constraintSetup [{targetBone {boneName=thigh_twist_01_r}}, {targetBone {boneName=pelvis}}]
    constraintWeights [1,1]
   }
   n15 = Get thigh_shape_fn_l()
   n16 = "Transform (Modify) Bone - Bone: thigh_volume_fn_l"(ComponentPose=@n14, Alpha=@n15, Translation="-5,-17,0", Rotation="0,0,0", Scale="1,1,1") {
    rotationMode=BMM_Additive
    rotationSpace=BCS_BoneSpace
    scaleMode=BMM_Additive
    scaleSpace=BCS_BoneSpace
    translationMode=BMM_Additive
    translationSpace=BCS_BoneSpace
    alphaBoolBlend {blendOption=Linear}
    boneToModify {boneName=thigh_volume_fn_l}
    translation=<-5 -17>
   }
   n17 = Get thigh_shape_fn_r()
   n18 = "Transform (Modify) Bone - Bone: thigh_volume_fn_r"(ComponentPose=@n16, Alpha=@n17, Translation="3,-10,0", Rotation="0,0,0", Scale="1,1,1") {
    rotationSpace=BCS_BoneSpace
    scaleMode=BMM_Additive
    scaleSpace=BCS_BoneSpace
    translationMode=BMM_Additive
    translationSpace=BCS_BoneSpace
    alphaBoolBlend {blendOption=Linear}
    boneToModify {boneName=thigh_volume_fn_r}
    translation=<3 -10>
   }
   n19 = "Transform (Modify) Bone - Bone: thigh_volume_bk_l"(ComponentPose=@n18, Alpha=@n15, Translation="-15,-2,0", Rotation="0,0,0", Scale="1.1,1.1,1.1") {
    rotationMode=BMM_Additive
    rotationSpace=BCS_BoneSpace
    scaleMode=BMM_Additive
    scaleSpace=BCS_BoneSpace
    translationMode=BMM_Additive
    translationSpace=BCS_BoneSpace
    alphaBoolBlend {blendOption=Linear}
    boneToModify {boneName=thigh_volume_bk_l}
    scale=<1.1 1.1 1.1>
    translation=<-15 -2>
   }
   n20 = "Transform (Modify) Bone - Bone: thigh_volume_bk_r"(ComponentPose=@n19, Alpha=@n17, Translation="15,-2,0", Rotation="0,0,0", Scale="1.1,1.1,1.1") {
    rotationSpace=BCS_BoneSpace
    scaleMode=BMM_Additive
    scaleSpace=BCS_BoneSpace
    translationMode=BMM_Additive
    translationSpace=BCS_BoneSpace
    alphaBoolBlend {blendOption=Linear}
    boneToModify {boneName=thigh_volume_bk_r}
    scale=<1.1 1.1 1.1>
    translation=<15 -2>
   }
   n21 = "Transform (Modify) Bone - Bone: thigh_belt_fn_l"(ComponentPose=@n20, Alpha=@n15, Translation="10,-5,0", Rotation="-15,30,0", Scale="1,1,1") {
    rotationMode=BMM_Additive
    rotationSpace=BCS_BoneSpace
    scaleMode=BMM_Additive
    scaleSpace=BCS_BoneSpace
    translationMode=BMM_Additive
    translationSpace=BCS_BoneSpace
    alphaBoolBlend {blendOption=Linear}
    boneToModify {boneName=thigh_belt_fn_l}
    rotation {pitch=-15, yaw=30}
    translation=<10 -5>
   }
   n22 = "Transform (Modify) Bone - Bone: thigh_belt_fn_r"(ComponentPose=@n21, Alpha=@n17, Translation="-10,-5,0", Rotation="-15,-30,0", Scale="1,1,1") {
    rotationMode=BMM_Additive
    rotationSpace=BCS_BoneSpace
    scaleMode=BMM_Additive
    scaleSpace=BCS_BoneSpace
    translationMode=BMM_Additive
    translationSpace=BCS_BoneSpace
    alphaBoolBlend {blendOption=Linear}
    boneToModify {boneName=thigh_belt_fn_r}
    rotation {pitch=-15, yaw=-30}
    translation=<-10 -5>
   }
   n23 = "Constraint - thigh_belt_fn_l (PP)"!(ComponentPose=@n22, Alpha=@n15, ConstraintWeights_0=0.7, ConstraintWeights_1:float=0.3) -> (Pose) {
    alphaBoolBlend {blendOption=Linear}
    boneToModify {boneName=thigh_belt_fn_l}
    constraintSetup [{targetBone {boneName=thigh_belt_fn_l}}, {targetBone {boneName=thigh_belt_fn_r}}]
    constraintWeights [1,1]
   }
   n24 = "Constraint - thigh_belt_fn_r (PP)"!(ComponentPose=@n23, Alpha=@n17, ConstraintWeights_0=0.3, ConstraintWeights_1:float=0.7) -> (Pose) {
    alphaBoolBlend {blendOption=Linear}
    boneToModify {boneName=thigh_belt_fn_r}
    constraintSetup [{targetBone {boneName=thigh_belt_fn_l}}, {targetBone {boneName=thigh_belt_fn_r}}]
    constraintWeights [1,1]
   }
   n25 = Get IsSkyDiving() -> (IsSkyDiving:double)
   n26 = "Transform (Modify) Bone - Bone: thigh_volume_bk_l"(ComponentPose=@n24, Alpha=@n25, Translation="-2,-3,0", Rotation="0,0,0", Scale="1,1,1") {
    rotationMode=BMM_Additive
    rotationSpace=BCS_BoneSpace
    scaleMode=BMM_Additive
    scaleSpace=BCS_BoneSpace
    translationMode=BMM_Additive
    translationSpace=BCS_BoneSpace
    alphaBoolBlend {blendOption=Linear}
    boneToModify {boneName=thigh_volume_bk_l}
    translation=<-2 -3>
   }
   n27 = "Transform (Modify) Bone - Bone: thigh_volume_bk_r"(ComponentPose=@n26, Alpha=@n25, Translation="-2,-3,0", Rotation="0,0,0", Scale="1,1,1") {
    rotationSpace=BCS_BoneSpace
    scaleMode=BMM_Additive
    scaleSpace=BCS_BoneSpace
    translationMode=BMM_Additive
    translationSpace=BCS_BoneSpace
    alphaBoolBlend {blendOption=Linear}
    boneToModify {boneName=thigh_volume_bk_r}
    translation=<-2 -3>
   }
   n28 = "Transform (Modify) Bone - Bone: thigh_volume_lt_l1"(ComponentPose=@n27, Alpha=Get thigh_shape_lt_l() -> (thigh_shape_lt_l:double), Translation="0,-10,0", Rotation="0,0,0", Scale="1.5,3,1.5") {
    rotationMode=BMM_Additive
    rotationSpace=BCS_BoneSpace
    scaleMode=BMM_Additive
    scaleSpace=BCS_BoneSpace
    translationMode=BMM_Additive
    translationSpace=BCS_BoneSpace
    alphaBoolBlend {blendOption=Linear}
    boneToModify {boneName=thigh_volume_lt_l1}
    scale=<1.5 3 1.5>
    translation {y=-10}
   }
   n29 = "Transform (Modify) Bone - Bone: thigh_volume_rt_r"(ComponentPose=@n28, Alpha=Get thigh_shape_rt_r() -> (thigh_shape_rt_r:double), Translation="0,-10,0", Rotation="0,0,0", Scale="1.5,3,1.5") {
    rotationSpace=BCS_BoneSpace
    scaleMode=BMM_Additive
    scaleSpace=BCS_BoneSpace
    translationMode=BMM_Additive
    translationSpace=BCS_BoneSpace
    alphaBoolBlend {blendOption=Linear}
    boneToModify {boneName=thigh_volume_rt_r}
    scale=<1.5 3 1.5>
    translation {y=-10}
   }
   n30 = "Constraint - calf_r_front (RR)"!(ComponentPose=@n29, Alpha=1, ConstraintWeights_0=0.5, ConstraintWeights_1:float=0.5) -> (Pose) {
    alphaBoolBlend {blendOption=Linear}
    boneToModify {boneName=calf_r_front}
    constraintSetup [{transformType=Rotation, targetBone {boneName=thigh_l}}, {transformType=Rotation, targetBone {boneName=calf_l}}]
    constraintWeights [1,1]
   }
   n31 = "Constraint - calf_r_front (RR)"(ComponentPose=@n30, Alpha=1, ConstraintWeights_0=0.5, ConstraintWeights_1=0.5) {
    alphaBoolBlend {blendOption=Linear}
    boneToModify {boneName=calf_r_front}
    constraintSetup [{transformType=Rotation, targetBone {boneName=thigh_r}}, {transformType=Rotation, targetBone {boneName=calf_r}}]
    constraintWeights [1,1]
   }
   n32 = "Constraint - calf_l_back (RR)"!(ComponentPose=@n31, Alpha=1, ConstraintWeights_0=0.5, ConstraintWeights_1:float=0.5) -> (Pose) {
    alphaBoolBlend {blendOption=Linear}
    boneToModify {boneName=calf_l_back}
    constraintSetup [{transformType=Rotation, targetBone {boneName=thigh_l}}, {transformType=Rotation, targetBone {boneName=calf_l}}]
    constraintWeights [1,1]
   }
   n33 = "Constraint - calf_r_back (RR)"!(ComponentPose=@n32, Alpha=1, ConstraintWeights_0=0.5, ConstraintWeights_1:float=0.5) -> (Pose) {
    alphaBoolBlend {blendOption=Linear}
    boneToModify {boneName=calf_r_back}
    constraintSetup [{transformType=Rotation, targetBone {boneName=thigh_r}}, {transformType=Rotation, targetBone {boneName=calf_r}}]
    constraintWeights [1,1]
   }
   n34 = Get calf_l_front() -> (calf_l_front:double)
   n35 = "Transform (Modify) Bone - Bone: calf_l_front"(ComponentPose=@n33, Alpha=@n34, Translation="0,1.5,0", Rotation="0,0,0", Scale="1,1,1") {
    rotationMode=BMM_Additive
    rotationSpace=BCS_BoneSpace
    scaleMode=BMM_Additive
    scaleSpace=BCS_BoneSpace
    translationMode=BMM_Additive
    translationSpace=BCS_BoneSpace
    alphaBoolBlend {blendOption=Linear}
    boneToModify {boneName=calf_l_front}
    translation {y=1.5}
   }
   n36 = Get calf_r_front() -> (calf_r_front:double)
   n37 = "Transform (Modify) Bone - Bone: calf_r_front"(ComponentPose=@n35, Alpha=@n36, Translation="0,-1.5,0", Rotation="0,0,0", Scale="1,1,1") {
    rotationMode=BMM_Additive
    rotationSpace=BCS_BoneSpace
    scaleMode=BMM_Additive
    scaleSpace=BCS_BoneSpace
    translationMode=BMM_Additive
    translationSpace=BCS_BoneSpace
    alphaBoolBlend {blendOption=Linear}
    boneToModify {boneName=calf_r_front}
    translation {y=-1.5}
   }
   n38 = "Transform (Modify) Bone - Bone: calf_l_back"(ComponentPose=@n37, Alpha=@n34, Translation="0,-5,0", Rotation="0,0,0", Scale="1,1,1") {
    rotationMode=BMM_Additive
    rotationSpace=BCS_BoneSpace
    scaleMode=BMM_Additive
    scaleSpace=BCS_BoneSpace
    translationMode=BMM_Additive
    translationSpace=BCS_BoneSpace
    alphaBoolBlend {blendOption=Linear}
    boneToModify {boneName=calf_l_back}
    translation {y=-5}
   }
   n39 = "Transform (Modify) Bone - Bone: calf_r_back"(ComponentPose=@n38, Alpha=@n36, Translation="0,5,0", Rotation="0,0,0", Scale="1,1,1") {
    rotationMode=BMM_Additive
    rotationSpace=BCS_BoneSpace
    scaleMode=BMM_Additive
    scaleSpace=BCS_BoneSpace
    translationMode=BMM_Additive
    translationSpace=BCS_BoneSpace
    alphaBoolBlend {blendOption=Linear}
    boneToModify {boneName=calf_r_back}
    translation {y=5}
   }
   n40 = "Constraint - flont_uppelalm_l (PP)"!(ComponentPose=@n39, Alpha=1, ConstraintWeights_0=0.5, ConstraintWeights_1:float=0.5) -> (Pose) {
    alphaBoolBlend {blendOption=Linear}
    boneToModify {boneName=flont_uppelalm_l}
    constraintSetup [{targetBone {boneName=upperarm_twist_01_l}}, {targetBone {boneName=clavicle_l}}]
    constraintWeights [1,1]
   }
   n41 = "Constraint - up_uppelalm_l (PP)"!(ComponentPose=@n40, Alpha=1, ConstraintWeights_0=0.5, ConstraintWeights_1:float=0.5) -> (Pose) {
    alphaBoolBlend {blendOption=Linear}
    boneToModify {boneName=up_uppelalm_l}
    constraintSetup [{targetBone {boneName=upperarm_twist_01_l}}, {targetBone {boneName=clavicle_l}}]
    constraintWeights [1,1]
   }
   n42 = "Constraint - front_upperarm_r (PP)"!(ComponentPose=@n41, Alpha=1, ConstraintWeights_0=0.5, ConstraintWeights_1:float=0.5) -> (Pose) {
    alphaBoolBlend {blendOption=Linear}
    boneToModify {boneName=front_upperarm_r}
    constraintSetup [{targetBone {boneName=upperarm_twist_01_r}}, {targetBone {boneName=clavicle_r}}]
    constraintWeights [1,1]
   }
   n43 = "Constraint - up_upperarm_r (PP)"!(ComponentPose=@n42, Alpha=1, ConstraintWeights_0=0.5, ConstraintWeights_1:float=0.5) -> (Pose) {
    alphaBoolBlend {blendOption=Linear}
    boneToModify {boneName=up_upperarm_r}
    constraintSetup [{targetBone {boneName=upperarm_twist_01_r}}, {targetBone {boneName=clavicle_r}}]
    constraintWeights [1,1]
   }
   n44 = "Transform (Modify) Bone - Bone: flont_uppelalm_l"(ComponentPose=@n43, Alpha=Get Shoulder_rot_f_l() -> (Shoulder_rot_f_l:double), Translation="-5,3,0", Rotation="0,0,0", Scale="1.5,1.5,1.5") {
    rotationMode=BMM_Additive
    rotationSpace=BCS_BoneSpace
    scaleMode=BMM_Additive
    scaleSpace=BCS_BoneSpace
    translationMode=BMM_Additive
    translationSpace=BCS_BoneSpace
    alphaBoolBlend {blendOption=Linear}
    boneToModify {boneName=flont_uppelalm_l}
    scale=<1.5 1.5 1.5>
    translation=<-5 3>
   }
   n45 = Get Shoulder_rot_f_r() -> (Shoulder_rot_f_r:double)
   n46 = "Transform (Modify) Bone - Bone: front_upperarm_r"(ComponentPose=@n44, Alpha=@n45, Translation="5,-3,0", Rotation="0,0,0", Scale="1.5,1.5,1.5") {
    rotationMode=BMM_Additive
    rotationSpace=BCS_BoneSpace
    scaleMode=BMM_Additive
    scaleSpace=BCS_BoneSpace
    translationMode=BMM_Additive
    translationSpace=BCS_BoneSpace
    alphaBoolBlend {blendOption=Linear}
    boneToModify {boneName=front_upperarm_r}
    scale=<1.5 1.5 1.5>
    translation=<5 -3>
   }
   n47 = "Transform (Modify) Bone - Bone: back_upperarm_r"(ComponentPose=@n46, Alpha=@n45, Translation="4,0,4", Rotation="0,0,0", Scale="1.5,1.5,1.5") {
    rotationMode=BMM_Additive
    rotationSpace=BCS_BoneSpace
    scaleMode=BMM_Additive
    scaleSpace=BCS_BoneSpace
    translationMode=BMM_Additive
    translationSpace=BCS_BoneSpace
    alphaBoolBlend {blendOption=Linear}
    boneToModify {boneName=back_upperarm_r}
    scale=<1.5 1.5 1.5>
    translation {x=4, z=4}
   }
   n48 = "Transform (Modify) Bone - Bone: up_uppelalm_l"(ComponentPose=@n47, Alpha=Get Shoulder_rot_up_l() -> (Shoulder_rot_up_l:double), Translation="0,0,0", Rotation="0,0,0", Scale="1.5,1.5,1.5") {
    rotationMode=BMM_Additive
    rotationSpace=BCS_BoneSpace
    scaleMode=BMM_Additive
    scaleSpace=BCS_BoneSpace
    translationMode=BMM_Additive
    translationSpace=BCS_BoneSpace
    alphaBoolBlend {blendOption=Linear}
    boneToModify {boneName=up_uppelalm_l}
    scale=<1.5 1.5 1.5>
   }
   n49 = "Transform (Modify) Bone - Bone: up_upperarm_r"(ComponentPose=@n48, Alpha=@n45, Translation="0,-3,5", Rotation="0,0,0", Scale="1.5,1.5,1.5") {
    rotationMode=BMM_Additive
    rotationSpace=BCS_BoneSpace
    scaleMode=BMM_Additive
    scaleSpace=BCS_BoneSpace
    translationMode=BMM_Additive
    translationSpace=BCS_BoneSpace
    alphaBoolBlend {blendOption=Linear}
    boneToModify {boneName=up_upperarm_r}
    scale=<1.5 1.5 1.5>
    translation {y=-3, z=5}
   }
   n50 = "Constraint - lowerarm_l_front (RR)"!(ComponentPose=@n49, Alpha=1, ConstraintWeights_0=0.5, ConstraintWeights_1:float=0.5) -> (Pose) {
    alphaBoolBlend {blendOption=Linear}
    boneToModify {boneName=lowerarm_l_front}
    constraintSetup [{transformType=Rotation, targetBone {boneName=lowerarm_l}}, {transformType=Rotation, targetBone {boneName=upperarm_l}}]
    constraintWeights [1,1]
   }
   n51 = "Constraint - lowerarm_r_front (RR)"!(ComponentPose=@n50, Alpha=1, ConstraintWeights_0=0.5, ConstraintWeights_1:float=0.5) -> (Pose) {
    alphaBoolBlend {blendOption=Linear}
    boneToModify {boneName=lowerarm_r_front}
    constraintSetup [{transformType=Rotation, targetBone {boneName=lowerarm_r}}, {transformType=Rotation, targetBone {boneName=upperarm_r}}]
    constraintWeights [1,1]
   }
   n52 = "Constraint - lowerarm_l_back (RR)"!(ComponentPose=@n51, Alpha=1, ConstraintWeights_0=0.5, ConstraintWeights_1:float=0.5) -> (Pose) {
    alphaBoolBlend {blendOption=Linear}
    boneToModify {boneName=lowerarm_l_back}
    constraintSetup [{transformType=Rotation, targetBone {boneName=lowerarm_l}}, {transformType=Rotation, targetBone {boneName=upperarm_l}}]
    constraintWeights [1,1]
   }
   n53 = "Constraint - lowerarm_r_back (RR)"!(ComponentPose=@n52, Alpha=1, ConstraintWeights_0=0.5, ConstraintWeights_1:float=0.5) -> (Pose) {
    alphaBoolBlend {blendOption=Linear}
    boneToModify {boneName=lowerarm_r_back}
    constraintSetup [{transformType=Rotation, targetBone {boneName=lowerarm_r}}, {transformType=Rotation, targetBone {boneName=upperarm_r}}]
    constraintWeights [1,1]
   }
   n54 = Get lowerarm_l_front() -> (lowerarm_l_front:double)
   n55 = "Transform (Modify) Bone - Bone: lowerarm_l_front"(ComponentPose=@n53, Alpha=@n54, Translation="0,5,0", Rotation="0,0,0", Scale="1,1,1") {
    rotationMode=BMM_Additive
    rotationSpace=BCS_BoneSpace
    scaleMode=BMM_Additive
    scaleSpace=BCS_BoneSpace
    translationMode=BMM_Additive
    translationSpace=BCS_BoneSpace
    alphaBoolBlend {blendOption=Linear}
    boneToModify {boneName=lowerarm_l_front}
    translation {y=5}
   }
   n56 = Get lowerarm_r_front() -> (lowerarm_r_front:double)
   n57 = "Transform (Modify) Bone - Bone: lowerarm_r_front"(ComponentPose=@n55, Alpha=@n56, Translation="0,-5,0", Rotation="0,0,0", Scale="1,1,1") {
    rotationMode=BMM_Additive
    rotationSpace=BCS_BoneSpace
    scaleMode=BMM_Additive
    scaleSpace=BCS_BoneSpace
    translationMode=BMM_Additive
    translationSpace=BCS_BoneSpace
    alphaBoolBlend {blendOption=Linear}
    boneToModify {boneName=lowerarm_r_front}
    translation {y=-5}
   }
   n58 = "Transform (Modify) Bone - Bone: lowerarm_l_back"(ComponentPose=@n57, Alpha=@n54, Translation="0,-2,0", Rotation="0,0,0", Scale="1,1,1") {
    rotationMode=BMM_Additive
    rotationSpace=BCS_BoneSpace
    scaleMode=BMM_Additive
    scaleSpace=BCS_BoneSpace
    translationMode=BMM_Additive
    translationSpace=BCS_BoneSpace
    alphaBoolBlend {blendOption=Linear}
    boneToModify {boneName=lowerarm_l_back}
    translation {y=-2}
   }
   n59 = "Transform (Modify) Bone - Bone: lowerarm_r_back"(ComponentPose=@n58, Alpha=@n56, Translation="0,2,0", Rotation="0,0,0", Scale="1,1,1") {
    rotationMode=BMM_Additive
    rotationSpace=BCS_BoneSpace
    scaleMode=BMM_Additive
    scaleSpace=BCS_BoneSpace
    translationMode=BMM_Additive
    translationSpace=BCS_BoneSpace
    alphaBoolBlend {blendOption=Linear}
    boneToModify {boneName=lowerarm_r_back}
    translation {y=2}
   }
   n60 = Get IsJump() -> (IsJump:double)
   n61 = "Transform (Modify) Bone - Bone: ball_Knife_r"(ComponentPose=@n59, Alpha=@n60, Translation="0,0,0", Rotation="0,0,0", Scale="0.01,0.01,0.01") {
    rotationMode=BMM_Additive
    rotationSpace=BCS_BoneSpace
    scaleMode=BMM_Additive
    scaleSpace=BCS_BoneSpace
    translationMode=BMM_Additive
    translationSpace=BCS_BoneSpace
    alphaBoolBlend {blendOption=Linear}
    boneToModify {boneName=ball_Knife_r}
    scale=<0.01 0.01 0.01>
   }
   n62 = "Transform (Modify) Bone - Bone: ball_Knife_l"(ComponentPose=@n61, Alpha=@n60, Translation="0,0,0", Rotation="0,0,0", Scale="0.01,0.01,0.01") {
    rotationMode=BMM_Additive
    rotationSpace=BCS_BoneSpace
    scaleMode=BMM_Additive
    scaleSpace=BCS_BoneSpace
    translationMode=BMM_Additive
    translationSpace=BCS_BoneSpace
    alphaBoolBlend {blendOption=Linear}
    boneToModify {boneName=ball_Knife_l}
    scale=<0.01 0.01 0.01>
   }
   n63 = Get Knife() -> (Knife:double)
   n64 = "Transform (Modify) Bone - Bone: hand_Knife01_r"(ComponentPose=@n62, Alpha=@n63, Translation="0,0,0", Rotation="0,0,0", Scale="0.01,0.01,0.01") {
    rotationMode=BMM_Additive
    rotationSpace=BCS_BoneSpace
    scaleMode=BMM_Additive
    scaleSpace=BCS_BoneSpace
    translationMode=BMM_Additive
    translationSpace=BCS_BoneSpace
    alphaBoolBlend {blendOption=Linear}
    boneToModify {boneName=hand_Knife01_r}
    scale=<0.01 0.01 0.01>
   }
   n65 = "Transform (Modify) Bone - Bone: hand_Knife02_r"(ComponentPose=@n64, Alpha=@n63, Translation="0,0,0", Rotation="0,0,0", Scale="0.01,0.01,0.01") {
    rotationMode=BMM_Additive
    rotationSpace=BCS_BoneSpace
    scaleMode=BMM_Additive
    scaleSpace=BCS_BoneSpace
    translationMode=BMM_Additive
    translationSpace=BCS_BoneSpace
    alphaBoolBlend {blendOption=Linear}
    boneToModify {boneName=hand_Knife02_r}
    scale=<0.01 0.01 0.01>
   }
   n66 = "Transform (Modify) Bone - Bone: hand_Knife01_l"(ComponentPose=@n65, Alpha=@n63, Translation="0,0,0", Rotation="0,0,0", Scale="0.01,0.01,0.01") {
    rotationMode=BMM_Additive
    rotationSpace=BCS_BoneSpace
    scaleMode=BMM_Additive
    scaleSpace=BCS_BoneSpace
    translationMode=BMM_Additive
    translationSpace=BCS_BoneSpace
    alphaBoolBlend {blendOption=Linear}
    boneToModify {boneName=hand_Knife01_l}
    scale=<0.01 0.01 0.01>
   }
   n67 = "Transform (Modify) Bone - Bone: hand_Knife02_l"(ComponentPose=@n66, Alpha=@n63, Translation="0,0,0", Rotation="0,0,0", Scale="0.01,0.01,0.01") {
    rotationMode=BMM_Additive
    rotationSpace=BCS_BoneSpace
    scaleMode=BMM_Additive
    scaleSpace=BCS_BoneSpace
    translationMode=BMM_Additive
    translationSpace=BCS_BoneSpace
    alphaBoolBlend {blendOption=Linear}
    boneToModify {boneName=hand_Knife02_l}
    scale=<0.01 0.01 0.01>
   }
   Output Pose(Result=Component To Local(ComponentPose=@n67)) {layerGroup=None}
   Get Shoulder_rot_up_r() -> (Shoulder_rot_up_r:double)
  ]}
 EventGraph {nodes [
   n1 = Event OnWeaponChanged()
   n2 = Get Dynamic Material Instance(execute=[@n1.then, @n22.then_0], SkeletalMeshComponent=GetPartSkeletalMeshComponent())
   > n3 = Sequence()
   n4 = PickaxeItemDefinitionHasAnyMetaTag(Tags=Make Array("[0]"=(TagName="Cosmetics.MetaTags.DarkAzalea")), bRequiresEquipped=true)
   > n5 = Branch(Condition=@n4)
   > n6 = Branch(Condition=Get bIsSkyDiving())
   Set Knife(execute=[@n8.then, @n6.then, @n7.else, @n5.else], Knife:double=1) -> (Output_Get:double)
   Set Knife(execute=[@n6.else, @n7.then], Knife=0)
   n7 = Branch(execute=[@n9.then, @n3.then_1], Condition=@n4)
   n8 = Event OnBeginSkydiving()
   n9 = Event OnEndSkydiving()
   Event BlueprintInitializeAnimation()
   > Set Skel_Mesh(Skel_Mesh=GetOwnerSkeletalMeshComponent())
   > n10 = Cast To PlayerPawn_Athena(Object=TryGetPawnOwner())
   > n11 = Set Pawn Owner(Pawn Owner:Fort Player Pawn Athena Object Reference=@n10) -> (Output_Get:Fort Player Pawn Athena Object Reference)
   > Set Shoulder_Mesh(Shoulder_Mesh:Skeletal Mesh Component Object Reference=GetSkeletalMeshForPartType(self=@n11)) -> (Output_Get:Skeletal Mesh Component Object Reference)
   Event BlueprintUpdateAnimation()
   > n12 = Sequence!(execute) -> (then_0, then_1, then_2:Exec, then_3:Exec)
   n13 = Get Skel_Mesh()
   > Set thigh_shape_fn_l(thigh_shape_fn_l:double=GetAlphaFromReferenceTransformDelta(SkeletalMeshComponent=@n13, BoneName=thigh_l, BaseName=pelvis, InputRange=(X=0,Y=-140), TwistAxis="0,0,1")) -> (Output_Get:double)
   > Set thigh_shape_fn_r(thigh_shape_fn_r:double=GetAlphaFromReferenceTransformDelta(SkeletalMeshComponent=@n13, BoneName=thigh_r, BaseName=pelvis, InputRange=(X=0,Y=-140), TwistAxis="0,0,1")) -> (Output_Get:double)
   n14 = Get Skel_Mesh()
   > Set thigh_shape_bk_l(thigh_shape_bk_l:double=GetAlphaFromReferenceTransformDelta(SkeletalMeshComponent=@n14, BoneName=thigh_l, BaseName=pelvis, InputRange=(X=0,Y=60), TwistAxis="0,0,1")) -> (Output_Get:double)
   > Set thigh_shape_bk_r(thigh_shape_bk_r:double=GetAlphaFromReferenceTransformDelta(SkeletalMeshComponent=@n14, BoneName=thigh_r, BaseName=pelvis, InputRange=(X=0,Y=60), TwistAxis="0,0,1")) -> (Output_Get:double)
   n15 = Get Skel_Mesh()
   > Set thigh_shape_lt_l(thigh_shape_lt_l:double=GetAlphaFromReferenceTransformDelta(SkeletalMeshComponent=@n15, BoneName=thigh_l, BaseName=pelvis, InputRange=(X=23,Y=-90), TwistAxis="0,0,1")) -> (Output_Get:double)
   > Set thigh_shape_rt_r(thigh_shape_rt_r:double=GetAlphaFromReferenceTransformDelta(SkeletalMeshComponent=@n15, BoneName=thigh_r, BaseName=pelvis, InputRange=(X=23,Y=-90), TwistAxis="0,0,1")) -> (Output_Get:double)
   n16 = Get Shoulder_Mesh() -> (Shoulder_Mesh:Skeletal Mesh Component Object Reference)
   > Set Shoulder_rot_f_r(Shoulder_rot_f_r:double=GetAlphaFromReferenceTransformDelta(SkeletalMeshComponent=@n16, BoneName=upperarm_r, BaseName=clavicle_r, InputRange=(X=0,Y=90), TwistAxis="0,0,1")) -> (Output_Get:double)
   > Set Shoulder_rot_f_l(Shoulder_rot_f_l:double=GetAlphaFromReferenceTransformDelta(SkeletalMeshComponent=@n16, BoneName=upperarm_l, BaseName=clavicle_l, InputRange=(X=0,Y=90), TwistAxis="0,0,1")) -> (Output_Get:double)
   n17 = Get Shoulder_Mesh()
   > Set Shoulder_rot_up_r(Shoulder_rot_up_r:double=GetAlphaFromReferenceTransformDelta(
     SkeletalMeshComponent=@n17
     BoneName=upperarm_twist_01_r
     BaseName=clavicle_r
     DeltaComponentType=EulerY
     InputRange=(X=0,Y=110)
     TwistAxis="0,0,1"
    )) -> (Output_Get:double)
   > Set Shoulder_rot_up_l(Shoulder_rot_up_l:double=GetAlphaFromReferenceTransformDelta(SkeletalMeshComponent=@n17, BoneName=upperarm_l, BaseName=clavicle_l, DeltaComponentType=EulerY, InputRange=(X=0,Y=110), TwistAxis="0,0,1")) -> (Output_Get:double)
   n18 = Get Shoulder_Mesh()
   > Set Shoulder_rot_dn_r(Shoulder_rot_dn_r:double=GetAlphaFromReferenceTransformDelta(SkeletalMeshComponent=@n18, BoneName=upperarm_r, BaseName=clavicle_r, InputRange=(X=-15,Y=-90), TwistAxis="0,0,1")) -> (Output_Get:double)
   > Set Shoulder_rot_dn_l(Shoulder_rot_dn_l:double=GetAlphaFromReferenceTransformDelta(SkeletalMeshComponent=@n18, BoneName=upperarm_l, BaseName=clavicle_l, InputRange=(X=-15,Y=-90), TwistAxis="0,0,1")) -> (Output_Get:double)
   > n19 = Sequence()
   > Set lowerarm_l_front(lowerarm_l_front:double=GetAlphaFromReferenceTransformDelta(SkeletalMeshComponent=Get Skel_Mesh(), BoneName=upperarm_l, BaseName=lowerarm_l, InputRange=(X=0,Y=-90), TwistAxis="0,0,1")) -> (Output_Get:double)
   > Set calf_l_front(calf_l_front:double=GetAlphaFromReferenceTransformDelta(SkeletalMeshComponent=Get Skel_Mesh(), BoneName=thigh_l, BaseName=calf_l, InputRange=(X=0,Y=-90), TwistAxis="0,0,1")) -> (Output_Get:double)
   Set lowerarm_r_front(
    execute=@n19.then_1
    lowerarm_r_front:double=GetAlphaFromReferenceTransformDelta(SkeletalMeshComponent=Get Skel_Mesh(), BoneName=upperarm_r, BaseName=lowerarm_r, InputRange=(X=0,Y=-90), TwistAxis="0,0,1")
   ) -> (Output_Get:double)
   > Set calf_r_front(calf_r_front:double=GetAlphaFromReferenceTransformDelta(SkeletalMeshComponent=Get Skel_Mesh(), BoneName=thigh_r, BaseName=calf_r, InputRange=(X=0,Y=-90), TwistAxis="0,0,1")) -> (Output_Get:double)
   Set neck_rot_up(
    execute=@n19.then_2
    neck_rot_up:double=GetAlphaFromReferenceTransformDelta(
     SkeletalMeshComponent=Get Shoulder_Mesh()
     BoneName=C_jaw
     BaseName=spine_05
     DeltaComponentType=EulerX
     InputRange=(X=30,Y=-30)
     TwistAxis="0,0,1"
    )
   ) -> (Output_Get:double)
   > Set neck_rot_dn(neck_rot_dn:double=GetAlphaFromReferenceTransformDelta(
     SkeletalMeshComponent=Get Shoulder_Mesh()
     BoneName=C_jaw
     BaseName=spine_05
     DeltaComponentType=EulerX
     InputRange=(X=0,Y=25)
     TwistAxis="0,0,1"
    )) -> (Output_Get:double)
   n20 = Branch(execute=@n12.then_1, Condition=Get bIsSkyDiving())
   > Set IsSkyDiving(IsSkyDiving:double=1) -> (Output_Get:double)
   Set IsSkyDiving(execute=@n20.else, IsSkyDiving=0)
   n21 = Branch(execute=@n12.then_2, Condition=OR Boolean!(A=Get bIsJumping(), B=Get bIsFalling(), C:bool=Get bIsSkyDiving(), D:bool) -> (ReturnValue))
   > Set IsJump(IsJump:double=0) -> (Output_Get:double)
   Set IsJump(execute=@n21.else, IsJump=1)
   Event OnCharacterCustomizationLoaded()
   > n22 = Sequence!(execute) -> (then_0, then_1)
   SetScalarParameterValue(execute=@n22.then_1, self=@n2, ParameterName=Hide Element 6_5, Value=0)
   > SetScalarParameterValue(self=@n2, ParameterName=Hide Element 01, Value=0)
   Reroute Node(InputPin:Material Instance Dynamic Object Reference=@n2) -> (OutputPin:Material Instance Dynamic Object Reference)
  ]}
]
variables [
 calf_l_front {type=number}
 calf_r_front {type=number}
 IsJump {type=number}
 IsSkyDiving {type=number}
 jump_falling {type=boolean}
 Knife {type=number}
 lowerarm_l_front {type=number}
 lowerarm_r_front {type=number}
 neck_rot_dn {type=number}
 neck_rot_up {type=number}
 Pawn Owner {type=FortPlayerPawnAthena}
 Shoulder_Mesh {type=SkeletalMeshComponent}
 Shoulder_rot_dn_l {type=number}
 Shoulder_rot_dn_r {type=number}
 Shoulder_rot_f_l {type=number}
 Shoulder_rot_f_r {type=number}
 Shoulder_rot_up_l {type=number}
 Shoulder_rot_up_r {type=number}
 Skel_Mesh {type=SkeletalMeshComponent}
 thigh_shape_bk_l {type=number}
 thigh_shape_bk_r {type=number}
 thigh_shape_fn_l {type=number}
 thigh_shape_fn_r {type=number}
 thigh_shape_lt_l {type=number}
 thigh_shape_rt_r {type=number}
]
