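# Post-process style AnimBlueprint for the Dark Azalea female character's FaceAcc mesh, driving secondary motion of its hair bones.
# Copies the body pose, then layers bone constraints and Anim Dynamics on side and back strands.
# Two RigidBody branches read BackpackCache and switch on isWearingPackBool; the pack branch offsets a collision bone and reshapes back hair.
# Diving tunes simulation response, while neck-rotation alphas drive corrective bone offsets before output.
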AnimBlueprint /BRCosmetics/Characters/Player/Female/Medium/Bodies/F_MED_DarkAzalea/Meshes/Parts/F_MED_Dark_Azalea_FaceAcc_AnimBP1
target_skeleton=&/F_MED_Dark_Azalea_FaceAcc_Skeleton
graphs [
 AnimGraph {nodes [
   Copy Pose From Mesh() {bUseAttachedParent=true}
   n1 = "Transform (Modify) Bone - Bone: dyn_Hair_BackL_04"!(ComponentPose, Alpha=1, Translation="-1.903,0.018,0.175", Rotation="0.000,0.000,5") -> (Pose) {
    rotationMode=BMM_Replace
    rotationSpace=BCS_BoneSpace
    scaleMode=BMM_Additive
    scaleSpace=BCS_BoneSpace
    translationMode=BMM_Additive
    translationSpace=BCS_BoneSpace
    alphaBoolBlend {blendOption=Linear}
    boneToModify {boneName=dyn_Hair_BackL_04}
    rotation {roll=5}
    scale {y=0.8}
    translation=<-1.903 0.018 0.175>
   }
   n2 = "Transform (Modify) Bone - Bone: dyn_Hair_Back_04"!(ComponentPose=@n1, Alpha=1, Translation="0,0,0", Rotation="0,0,0") -> (Pose) {
    rotationMode=BMM_Replace
    rotationSpace=BCS_BoneSpace
    scaleMode=BMM_Additive
    scaleSpace=BCS_BoneSpace
    translationMode=BMM_Additive
    translationSpace=BCS_BoneSpace
    alphaBoolBlend {blendOption=Linear}
    boneToModify {boneName=dyn_Hair_Back_04}
   }
   n3 = "Transform (Modify) Bone - Bone: dyn_Hair_BackR_04"!(ComponentPose=@n2, Alpha=1, Translation="0.5,-0.000,0", Rotation="0,15,0") -> (Pose) {
    rotationMode=BMM_Replace
    rotationSpace=BCS_BoneSpace
    scaleMode=BMM_Additive
    scaleSpace=BCS_BoneSpace
    translationMode=BMM_Additive
    translationSpace=BCS_BoneSpace
    alphaBoolBlend {blendOption=Linear}
    boneToModify {boneName=dyn_Hair_BackR_04}
    rotation {yaw=15}
    scale {y=0.8}
    translation {x=0.5}
   }
   n4 = "Transform (Modify) Bone - Bone: dyn_Hair_BackL_05"!(ComponentPose=@n3, Alpha=1, Translation="0,0,0", Rotation="0,-10,0") -> (Pose) {
    rotationMode=BMM_Replace
    rotationSpace=BCS_BoneSpace
    scaleMode=BMM_Additive
    scaleSpace=BCS_BoneSpace
    translationMode=BMM_Additive
    translationSpace=BCS_BoneSpace
    alphaBoolBlend {blendOption=Linear}
    boneToModify {boneName=dyn_Hair_BackL_05}
    rotation {yaw=-10}
    scale {y=0.4}
   }
   n5 = "Transform (Modify) Bone - Bone: dyn_Hair_Back_05"!(ComponentPose=@n4, Alpha=1, Translation="0,0,0", Rotation="0,-10,0") -> (Pose) {
    rotationMode=BMM_Replace
    rotationSpace=BCS_BoneSpace
    scaleMode=BMM_Additive
    scaleSpace=BCS_BoneSpace
    translationMode=BMM_Additive
    translationSpace=BCS_BoneSpace
    alphaBoolBlend {blendOption=Linear}
    boneToModify {boneName=dyn_Hair_Back_05}
    rotation {yaw=-10}
    scale {y=0.4}
   }
   n6 = "Transform (Modify) Bone - Bone: dyn_Hair_BackR_05"!(ComponentPose=@n5, Alpha=1, Translation="0,0,-6", Rotation="0.187,-10,-13.726") -> (Pose) {
    rotationMode=BMM_Replace
    rotationSpace=BCS_BoneSpace
    scaleMode=BMM_Additive
    scaleSpace=BCS_BoneSpace
    translationMode=BMM_Additive
    translationSpace=BCS_BoneSpace
    alphaBoolBlend {blendOption=Linear}
    boneToModify {boneName=dyn_Hair_BackR_05}
    rotation=<rot 0.187 -10 -13.726>
    scale {y=0.4}
    translation {z=-6}
   }
   n7 = "Transform (Modify) Bone - Bone: dyn_Hair_BackL_07"!(ComponentPose=@n6, Alpha=1, Translation="1.273,-0.321,5.165") -> (Pose) {
    rotationMode=BMM_Replace
    rotationSpace=BCS_BoneSpace
    scaleMode=BMM_Additive
    scaleSpace=BCS_BoneSpace
    translationMode=BMM_Additive
    translationSpace=BCS_BoneSpace
    alphaBoolBlend {blendOption=Linear}
    boneToModify {boneName=dyn_Hair_BackL_07}
    scale {y=0.1}
    translation=<1.273 -0.321 5.165>
   }
   n8 = "Transform (Modify) Bone - Bone: dyn_Hair_Back_07"!(ComponentPose=@n7, Alpha=1, Translation="0.005,3.156,0.365") -> (Pose) {
    rotationMode=BMM_Replace
    rotationSpace=BCS_BoneSpace
    scaleMode=BMM_Additive
    scaleSpace=BCS_BoneSpace
    translationMode=BMM_Additive
    translationSpace=BCS_BoneSpace
    alphaBoolBlend {blendOption=Linear}
    boneToModify {boneName=dyn_Hair_Back_07}
    scale {y=0.1}
    translation=<0.005 3.156 0.365>
   }
   "Transform (Modify) Bone - Bone: dyn_Hair_BackR_07"!(ComponentPose=@n8, Alpha=1, Translation="0.027,-0.011,-7.192") -> (Pose) {
    rotationMode=BMM_Replace
    rotationSpace=BCS_BoneSpace
    scaleMode=BMM_Additive
    scaleSpace=BCS_BoneSpace
    translationMode=BMM_Additive
    translationSpace=BCS_BoneSpace
    alphaBoolBlend {blendOption=Linear}
    boneToModify {boneName=dyn_Hair_BackR_07}
    scale {y=0.1}
    translation=<0.027 -0.011 -7.192>
   }
   n9 = "Transform (Modify) Bone - Bone: dyn_Hair_r_base"(Alpha=Get Shoulder_rot_f_r() -> (Shoulder_rot_f_r:double), Translation="0,0,0", Rotation="40,5,0", Scale="1,1,1") {
    rotationMode=BMM_Additive
    rotationSpace=BCS_BoneSpace
    scaleMode=BMM_Additive
    scaleSpace=BCS_BoneSpace
    translationMode=BMM_Additive
    translationSpace=BCS_BoneSpace
    alphaBoolBlend {blendOption=Linear}
    boneToModify {boneName=dyn_Hair_r_base}
    rotation {pitch=40, yaw=5}
   }
   "Transform (Modify) Bone - Bone: dyn_Hair_r_base"(ComponentPose=@n9, Alpha=Get Shoulder_rot_up_r() -> (Shoulder_rot_up_r:double), Translation="0,0,0", Rotation="0,-30,0", Scale="1,1,1") {
    rotationMode=BMM_Additive
    rotationSpace=BCS_BoneSpace
    scaleMode=BMM_Additive
    scaleSpace=BCS_BoneSpace
    translationMode=BMM_Additive
    translationSpace=BCS_BoneSpace
    alphaBoolBlend {blendOption=Linear}
    boneToModify {boneName=dyn_Hair_r_base}
    rotation {yaw=-30}
   }
   n10 = Get Diving() -> (Diving:bool)
   n11 = RigidBody!(
    ComponentPose=Local To Component(LocalPose=Use cached pose 'BackpackCache'())
    Alpha
    OverrideWorldGravity:Vector=Get Gravity Override Cape() -> (Gravity Override Cape:Vector)
    ExternalForce
    ComponentLinearAccScale:Vector=SelectVector(A="0.01,0.01,0.01", B="0.35,0.35,0.1", bPickA=@n10)
    ComponentLinearVelScale:Vector=SelectVector(A="0.01,0.01,0.01", B="0.35,0.35,0.1", bPickA=@n10)
   ) -> (Pose) {
    bClampLinearTranslationLimitToRefPose=true
    bOverrideWorldGravity=true
    overridePhysicsAsset=&/F_MED_Dark_Azalea_FaceAcc_Physics
    simulationSpace=BaseBoneSpace
    alphaBoolBlend {blendOption=Linear}
    baseBoneRef {boneName=neck_01}
    componentAppliedLinearAccClamp=<2000 2000 2000>
    componentLinearAccScale=<0.3 0.3 0.35>
    componentLinearVelScale=<0.35 0.35 0.35>
    overrideWorldGravity {z=-980}
    simSpaceSettings {
     maxAngularAcceleration=1000
     maxAngularVelocity=0.3
     maxLinearAcceleration=700
     maxLinearVelocity=500
     velocityScaleZ=0
     worldAlpha=0.1
     externalLinearDragV=<0.5 0.5>
    }
   }
   n12 = Component To Local(ComponentPose=@n11)
   n13 = "Transform (Modify) Bone - Bone: backpack_collision"(
    ComponentPose=Local To Component(LocalPose=Use cached pose 'BackpackCache'())
    Alpha=1
    Translation="-0.003,100,-0.022"
    Rotation="0,0,0"
    Scale="1,1,1"
   ) {
    rotationMode=BMM_Additive
    rotationSpace=BCS_BoneSpace
    scaleMode=BMM_Additive
    scaleSpace=BCS_BoneSpace
    translationMode=BMM_Additive
    translationSpace=BCS_BoneSpace
    alphaBoolBlend {blendOption=Linear}
    boneToModify {boneName=backpack_collision}
    translation=<-0.003 100 -0.022>
   }
   n14 = Get Diving()
   n15 = RigidBody(
    ComponentPose=@n13
    OverrideWorldGravity=SelectVector(A="0, 0,-750", B="0, 0,-980", bPickA=@n14)
    ComponentLinearAccScale=SelectVector(A="0.05,0.05,0.05", B="0.1,0.1,0.1", bPickA=@n14)
    ComponentLinearVelScale=SelectVector(A="0.055,0.055,0.055", B="0.1,0.1,0.1", bPickA=@n14)
   ) {
    bClampLinearTranslationLimitToRefPose=true
    bOverrideWorldGravity=true
    overridePhysicsAsset=&F_MED_Dark_Azalea_FaceAcc_Physics
    simulationSpace=BaseBoneSpace
    alphaBoolBlend {blendOption=Linear}
    baseBoneRef {boneName=neck_01}
    componentAppliedLinearAccClamp=<200 200 200>
    componentLinearAccScale=<0.1 0.1 0.1>
    componentLinearVelScale=<0.1 0.1 0.1>
    overrideWorldGravity {z=-980}
    simSpaceSettings {
     maxAngularAcceleration=1000
     maxAngularVelocity=0.3
     maxLinearAcceleration=700
     maxLinearVelocity=500
     velocityScaleZ=0
     worldAlpha=0.1
     externalLinearDragV=<0.5 0.5>
    }
   }
   n16 = "Constraint - dyn_Hair_BackL_05 (PP)"!(ComponentPose=@n15, Alpha=1, ConstraintWeights_0=0.3, ConstraintWeights_1:float=0.7) -> (Pose) {
    alphaBoolBlend {blendOption=Linear}
    boneToModify {boneName=dyn_Hair_BackL_05}
    constraintSetup [{targetBone {boneName=spine_05}}, {targetBone {boneName=dyn_Hair_Back_05}}]
    constraintWeights [1,1]
   }
   n17 = "Constraint - dyn_Hair_BackL_06 (PP)"!(ComponentPose=@n16, Alpha=1, ConstraintWeights_0=0.3, ConstraintWeights_1:float=0.7) -> (Pose) {
    alphaBoolBlend {blendOption=Linear}
    boneToModify {boneName=dyn_Hair_BackL_06}
    constraintSetup [{targetBone {boneName=spine_05}}, {targetBone {boneName=dyn_Hair_Back_06}}]
    constraintWeights [1,1]
   }
   n18 = "Constraint - dyn_Hair_BackL_07 (PP)"!(ComponentPose=@n17, Alpha=1, ConstraintWeights_0=0.7, ConstraintWeights_1:float=0.3) -> (Pose) {
    alphaBoolBlend {blendOption=Linear}
    boneToModify {boneName=dyn_Hair_BackL_07}
    constraintSetup [{targetBone {boneName=spine_05}}, {targetBone {boneName=dyn_Hair_Back_07}}]
    constraintWeights [1,1]
   }
   n19 = "Constraint - dyn_Hair_BackL_08 (PP)"!(ComponentPose=@n18, Alpha=1, ConstraintWeights_0=0.7, ConstraintWeights_1:float=0.3) -> (Pose) {
    alphaBoolBlend {blendOption=Linear}
    boneToModify {boneName=dyn_Hair_BackL_08}
    constraintSetup [{targetBone {boneName=spine_05}}, {targetBone {boneName=dyn_Hair_Back_08}}]
    constraintWeights [1,1]
   }
   n20 = "Constraint - dyn_Hair_BackR_05 (PP)"!(ComponentPose=@n19, Alpha=1, ConstraintWeights_0=0.3, ConstraintWeights_1:float=0.7) -> (Pose) {
    alphaBoolBlend {blendOption=Linear}
    boneToModify {boneName=dyn_Hair_BackR_05}
    constraintSetup [{targetBone {boneName=spine_05}}, {targetBone {boneName=dyn_Hair_Back_05}}]
    constraintWeights [1,1]
   }
   n21 = "Constraint - dyn_Hair_BackR_06 (PP)"!(ComponentPose=@n20, Alpha=1, ConstraintWeights_0=0.3, ConstraintWeights_1:float=0.7) -> (Pose) {
    alphaBoolBlend {blendOption=Linear}
    boneToModify {boneName=dyn_Hair_BackR_06}
    constraintSetup [{targetBone {boneName=spine_05}}, {targetBone {boneName=dyn_Hair_Back_06}}]
    constraintWeights [1,1]
   }
   n22 = "Constraint - dyn_Hair_BackR_07 (PP)"!(ComponentPose=@n21, Alpha=1, ConstraintWeights_0=0.7, ConstraintWeights_1:float=0.3) -> (Pose) {
    alphaBoolBlend {blendOption=Linear}
    boneToModify {boneName=dyn_Hair_BackR_07}
    constraintSetup [{targetBone {boneName=spine_05}}, {targetBone {boneName=dyn_Hair_Back_07}}]
    constraintWeights [1,1]
   }
   n23 = "Constraint - dyn_Hair_BackR_08 (PP)"!(ComponentPose=@n22, Alpha=1, ConstraintWeights_0=0.7, ConstraintWeights_1:float=0.3) -> (Pose) {
    alphaBoolBlend {blendOption=Linear}
    boneToModify {boneName=dyn_Hair_BackR_08}
    constraintSetup [{targetBone {boneName=spine_05}}, {targetBone {boneName=dyn_Hair_Back_08}}]
    constraintWeights [1,1]
   }
   n24 = "Constraint - dyn_Hair_BackL_05 (PPP)"!(ComponentPose=@n23, Alpha=1, ConstraintWeights_0=0.0001, ConstraintWeights_1:float=0.107, ConstraintWeights_2:float=0.8929) -> (Pose) {
    alphaBoolBlend {blendOption=Linear}
    boneToModify {boneName=dyn_Hair_BackL_05}
    constraintSetup [{targetBone {boneName=upperarm_l}}, {targetBone {boneName=clavicle_l}}, {targetBone {boneName=spine_05}}]
    constraintWeights [1,1,1]
   }
   n25 = "Constraint - dyn_Hair_Back_05 (PPPPP)"!(
    ComponentPose=@n24
    Alpha=1
    ConstraintWeights_0=0
    ConstraintWeights_1:float=0.0131
    ConstraintWeights_2:float=0
    ConstraintWeights_3:float=0.0084
    ConstraintWeights_4:float=0.9786
   ) -> (Pose) {
    alphaBoolBlend {blendOption=Linear}
    boneToModify {boneName=dyn_Hair_Back_05}
    constraintSetup [
     {targetBone {boneName=upperarm_r}}
     {targetBone {boneName=clavicle_r}}
     {targetBone {boneName=upperarm_l}}
     {targetBone {boneName=clavicle_l}}
     {targetBone {boneName=spine_05}}
    ]
    constraintWeights [1,1,1,1,1]
   }
   n26 = "Constraint - dyn_Hair_BackR_05 (PPP)"!(ComponentPose=@n25, Alpha=1, ConstraintWeights_0=0.0062, ConstraintWeights_1:float=0.1661, ConstraintWeights_2:float=0.8277) -> (Pose) {
    alphaBoolBlend {blendOption=Linear}
    boneToModify {boneName=dyn_Hair_BackR_05}
    constraintSetup [{targetBone {boneName=upperarm_l}}, {targetBone {boneName=clavicle_r}}, {targetBone {boneName=spine_05}}]
    constraintWeights [1,1,1]
   }
   n27 = "Transform (Modify) Bone - Bone: dyn_Hair_BackL_04"(ComponentPose=@n26, Alpha=1, Translation="-1.903,0.018,0.175", Rotation="0.000,0.000,5") {
    rotationMode=BMM_Replace
    rotationSpace=BCS_BoneSpace
    scaleMode=BMM_Additive
    scaleSpace=BCS_BoneSpace
    translationMode=BMM_Additive
    translationSpace=BCS_BoneSpace
    alphaBoolBlend {blendOption=Linear}
    boneToModify {boneName=dyn_Hair_BackL_04}
    rotation {roll=5}
    scale {y=0.8}
    translation=<-1.903 0.018 0.175>
   }
   n28 = "Transform (Modify) Bone - Bone: dyn_Hair_Back_04"(ComponentPose=@n27, Alpha=1, Translation="0,0,0", Rotation="0,0,0") {
    rotationMode=BMM_Replace
    rotationSpace=BCS_BoneSpace
    scaleMode=BMM_Additive
    scaleSpace=BCS_BoneSpace
    translationMode=BMM_Additive
    translationSpace=BCS_BoneSpace
    alphaBoolBlend {blendOption=Linear}
    boneToModify {boneName=dyn_Hair_Back_04}
   }
   n29 = "Transform (Modify) Bone - Bone: dyn_Hair_BackR_04"(ComponentPose=@n28, Alpha=1, Translation="0.5,-0.000,0", Rotation="0,15,0") {
    rotationMode=BMM_Replace
    rotationSpace=BCS_BoneSpace
    scaleMode=BMM_Additive
    scaleSpace=BCS_BoneSpace
    translationMode=BMM_Additive
    translationSpace=BCS_BoneSpace
    alphaBoolBlend {blendOption=Linear}
    boneToModify {boneName=dyn_Hair_BackR_04}
    rotation {yaw=15}
    scale {y=0.8}
    translation {x=0.5}
   }
   n30 = "Transform (Modify) Bone - Bone: dyn_Hair_BackL_05"(ComponentPose=@n29, Alpha=1, Translation="0,0,0", Rotation="-0.324,-14.676,1.163") {
    rotationMode=BMM_Replace
    rotationSpace=BCS_BoneSpace
    scaleMode=BMM_Additive
    scaleSpace=BCS_BoneSpace
    translationMode=BMM_Additive
    translationSpace=BCS_BoneSpace
    alphaBoolBlend {blendOption=Linear}
    boneToModify {boneName=dyn_Hair_BackL_05}
    rotation=<rot -0.324 -14.676 1.163>
    scale {y=0.4}
   }
   n31 = "Transform (Modify) Bone - Bone: dyn_Hair_Back_05"(ComponentPose=@n30, Alpha=1, Translation="0,0,0", Rotation="-0.110,-32.970,0.037") {
    rotationMode=BMM_Replace
    rotationSpace=BCS_BoneSpace
    scaleMode=BMM_Additive
    scaleSpace=BCS_BoneSpace
    translationMode=BMM_Additive
    translationSpace=BCS_BoneSpace
    alphaBoolBlend {blendOption=Linear}
    boneToModify {boneName=dyn_Hair_Back_05}
    rotation=<rot -0.11 -32.97 0.037>
    scale {y=0.4}
   }
   n32 = "Transform (Modify) Bone - Bone: dyn_Hair_BackR_05"(ComponentPose=@n31, Alpha=1, Translation="0,0,-6", Rotation="-1.322,-22.950,1.649") {
    rotationMode=BMM_Replace
    rotationSpace=BCS_BoneSpace
    scaleMode=BMM_Additive
    scaleSpace=BCS_BoneSpace
    translationMode=BMM_Additive
    translationSpace=BCS_BoneSpace
    alphaBoolBlend {blendOption=Linear}
    boneToModify {boneName=dyn_Hair_BackR_05}
    rotation=<rot -1.322 -22.95 1.649>
    scale {y=0.4}
    translation {z=-6}
   }
   n33 = "Transform (Modify) Bone - Bone: dyn_Hair_BackL_07"(ComponentPose=@n32, Alpha=1, Translation="0,0,5") {
    rotationMode=BMM_Replace
    rotationSpace=BCS_BoneSpace
    scaleMode=BMM_Additive
    scaleSpace=BCS_BoneSpace
    translationMode=BMM_Additive
    translationSpace=BCS_BoneSpace
    alphaBoolBlend {blendOption=Linear}
    boneToModify {boneName=dyn_Hair_BackL_07}
    scale {y=0.1}
    translation {z=5}
   }
   n34 = "Transform (Modify) Bone - Bone: dyn_Hair_Back_07"(ComponentPose=@n33, Alpha=1, Translation="0,5,0") {
    rotationMode=BMM_Replace
    rotationSpace=BCS_BoneSpace
    scaleMode=BMM_Additive
    scaleSpace=BCS_BoneSpace
    translationMode=BMM_Additive
    translationSpace=BCS_BoneSpace
    alphaBoolBlend {blendOption=Linear}
    boneToModify {boneName=dyn_Hair_Back_07}
    scale {y=0.1}
    translation {y=5}
   }
   n35 = "Transform (Modify) Bone - Bone: dyn_Hair_BackR_07"(ComponentPose=@n34, Alpha=1, Translation="0,0,-5") {
    rotationMode=BMM_Replace
    rotationSpace=BCS_BoneSpace
    scaleMode=BMM_Additive
    scaleSpace=BCS_BoneSpace
    translationMode=BMM_Additive
    translationSpace=BCS_BoneSpace
    alphaBoolBlend {blendOption=Linear}
    boneToModify {boneName=dyn_Hair_BackR_07}
    rotation=<rot -8.8677 -9.3482 -3.1888>
    scale {y=0.1}
    translation {z=-5}
   }
   n36 = "Transform (Modify) Bone - Bone: dyn_Hair_BackL_04"(ComponentPose=@n35, Alpha=1, Translation="0.5,-2.5,1.5", Rotation="0,0,0") {
    rotationMode=BMM_Replace
    rotationSpace=BCS_BoneSpace
    scaleMode=BMM_Additive
    scaleSpace=BCS_BoneSpace
    translationMode=BMM_Additive
    translationSpace=BCS_BoneSpace
    alphaBoolBlend {blendOption=Linear}
    boneToModify {boneName=dyn_Hair_BackL_04}
    translation=<0.5 -2.5 1.5>
   }
   n37 = "Transform (Modify) Bone - Bone: dyn_Hair_Back_04"(ComponentPose=@n36, Alpha=1, Translation="1,0,0", Rotation="0,0,0") {
    rotationMode=BMM_Replace
    rotationSpace=BCS_BoneSpace
    scaleMode=BMM_Additive
    scaleSpace=BCS_BoneSpace
    translationMode=BMM_Additive
    translationSpace=BCS_BoneSpace
    alphaBoolBlend {blendOption=Linear}
    boneToModify {boneName=dyn_Hair_Back_04}
    translation {x=1}
   }
   n38 = "Transform (Modify) Bone - Bone: dyn_Hair_BackR_04"(ComponentPose=@n37, Alpha=1, Translation="0.5,-2,1", Rotation="0,0,0") {
    rotationMode=BMM_Replace
    rotationSpace=BCS_BoneSpace
    scaleMode=BMM_Additive
    scaleSpace=BCS_BoneSpace
    translationMode=BMM_Additive
    translationSpace=BCS_BoneSpace
    alphaBoolBlend {blendOption=Linear}
    boneToModify {boneName=dyn_Hair_BackR_04}
    scale {y=0.8}
    translation=<0.5 -2 1>
   }
   n39 = Component To Local(ComponentPose=@n38)
   n40 = Get isWearingPackBool() -> (isWearingPackBool:bool)
   n41 = Get neck_rot_up() -> (neck_rot_up:double)
   n42 = "Transform (Modify) Bone - Bone: dyn_Hair_Back_02"(
    ComponentPose=Local To Component(LocalPose=Blend Poses by bool(BlendPose_0=@n39, BlendPose_1=@n12, bActiveValue=@n40) {blendType=Linear})
    Alpha=@n41
    Translation="-1,3,0"
    Rotation="0,0,0"
    Scale="1,1,1"
   ) {
    rotationMode=BMM_Additive
    rotationSpace=BCS_BoneSpace
    scaleMode=BMM_Additive
    scaleSpace=BCS_BoneSpace
    translationMode=BMM_Additive
    translationSpace=BCS_BoneSpace
    alphaBoolBlend {blendOption=Linear}
    boneToModify {boneName=dyn_Hair_Back_02}
    translation=<-1 3>
   }
   n43 = "Transform (Modify) Bone - Bone: dyn_Hair_BackL_02"(ComponentPose=@n42, Alpha=@n41, Translation="-1,3,0", Rotation="0,0,0", Scale="1,1,1") {
    rotationMode=BMM_Additive
    rotationSpace=BCS_BoneSpace
    scaleMode=BMM_Additive
    scaleSpace=BCS_BoneSpace
    translationMode=BMM_Additive
    translationSpace=BCS_BoneSpace
    alphaBoolBlend {blendOption=Linear}
    boneToModify {boneName=dyn_Hair_BackL_02}
    translation=<-1 3>
   }
   n44 = "Transform (Modify) Bone - Bone: dyn_Hair_BackR_02"(ComponentPose=@n43, Alpha=@n41, Translation="-2,7,0", Rotation="0,0,0", Scale="1,1,1") {
    rotationMode=BMM_Additive
    rotationSpace=BCS_BoneSpace
    scaleMode=BMM_Additive
    scaleSpace=BCS_BoneSpace
    translationMode=BMM_Additive
    translationSpace=BCS_BoneSpace
    alphaBoolBlend {blendOption=Linear}
    boneToModify {boneName=dyn_Hair_BackR_02}
    translation=<-2 7>
   }
   n45 = "Transform (Modify) Bone - Bone: dyn_Hair_Back_04"!(ComponentPose=@n44, Alpha=@n41, Translation="0,0,0", Rotation="0,-7,0", Scale:Vector="1,1,1") -> (Pose) {
    rotationMode=BMM_Additive
    rotationSpace=BCS_BoneSpace
    scaleMode=BMM_Additive
    scaleSpace=BCS_BoneSpace
    translationMode=BMM_Additive
    translationSpace=BCS_BoneSpace
    alphaBoolBlend {blendOption=Linear}
    boneToModify {boneName=dyn_Hair_Back_04}
    rotation {yaw=-7}
   }
   n46 = "Transform (Modify) Bone - Bone: dyn_Hair_BackL_04"!(ComponentPose=@n45, Alpha=@n41, Translation="0,0,0", Rotation="0,-7,0", Scale:Vector="1,1,1") -> (Pose) {
    rotationMode=BMM_Additive
    rotationSpace=BCS_BoneSpace
    scaleMode=BMM_Additive
    scaleSpace=BCS_BoneSpace
    translationMode=BMM_Additive
    translationSpace=BCS_BoneSpace
    alphaBoolBlend {blendOption=Linear}
    boneToModify {boneName=dyn_Hair_BackL_04}
    rotation {yaw=-7}
   }
   n47 = "Transform (Modify) Bone - Bone: dyn_Hair_BackR_04"!(ComponentPose=@n46, Alpha=@n41, Translation="0,0,0", Rotation="0,-7,0", Scale:Vector="1,1,1") -> (Pose) {
    rotationMode=BMM_Additive
    rotationSpace=BCS_BoneSpace
    scaleMode=BMM_Additive
    scaleSpace=BCS_BoneSpace
    translationMode=BMM_Additive
    translationSpace=BCS_BoneSpace
    alphaBoolBlend {blendOption=Linear}
    boneToModify {boneName=dyn_Hair_BackR_04}
    rotation {yaw=-7}
   }
   n48 = "Transform (Modify) Bone - Bone: dyn_Hair_BackL_02"!(ComponentPose=@n47, bAlphaBoolEnabled:bool=Get Diving(), Translation="-1,3,0", Rotation="0,0,0", Scale="1,1,1") -> (Pose) {
    alphaInputType=Bool
    rotationMode=BMM_Additive
    rotationSpace=BCS_BoneSpace
    scaleMode=BMM_Additive
    scaleSpace=BCS_BoneSpace
    translationMode=BMM_Additive
    translationSpace=BCS_BoneSpace
    alphaBoolBlend {blendOption=Linear}
    boneToModify {boneName=dyn_Hair_BackL_02}
    translation=<-1 3>
   }
   Output Pose(Result=Component To Local(ComponentPose=@n48)) {layerGroup=None}
   n49 = Get Skel_Mesh()
   n50 = "Constraint - dyn_Hair_Back_02 (RR)"!(
    ComponentPose=Local To Component(LocalPose=Copy Pose From Mesh(SourceMeshComponent=@n49) {bUseAttachedParent=true})
    Alpha=1
    ConstraintWeights_0=0.5
    ConstraintWeights_1:float=0.5
   ) -> (Pose) {
    alphaBoolBlend {blendOption=Linear}
    boneToModify {boneName=dyn_Hair_Back_02}
    constraintSetup [{transformType=Rotation, targetBone {boneName=spine_05}}, {transformType=Rotation, targetBone {boneName=dyn_Hair_Back_02}}]
    constraintWeights [1,1]
   }
   n51 = "Constraint - dyn_Hair_l_04 (PP)"!(ComponentPose=@n50, Alpha=1, ConstraintWeights_0=1, ConstraintWeights_1:float=0) -> (Pose) {
    alphaBoolBlend {blendOption=Linear}
    boneToModify {boneName=dyn_Hair_l_04}
    constraintSetup [{targetBone {boneName=spine_05}}, {targetBone {boneName=dyn_Hair_l_02}}]
    constraintWeights [1,1]
   }
   n52 = "Constraint - dyn_Hair_r_03 (RR)"!(ComponentPose=@n51, Alpha=1, ConstraintWeights_0=0.8, ConstraintWeights_1:float=0.2) -> (Pose) {
    alphaBoolBlend {blendOption=Linear}
    boneToModify {boneName=dyn_Hair_r_03}
    constraintSetup [{transformType=Rotation, targetBone {boneName=spine_05}}, {transformType=Rotation, targetBone {boneName=dyn_Hair_r_02}}]
    constraintWeights [1,1]
   }
   n53 = Get Body_Enabled_0_5() -> (Body_Enabled_0_5:double)
   n54 = Anim Dynamics - Bone: dyn_Hair_l_01(ComponentPose=@n52, Alpha=@n53, ExternalForce="0,0,0") {
    angularDampingOverride=0.999
    bAngularSpring=true
    bGravityOverrideInSimSpace=true
    bOverrideAngularDamping=true
    bUseGravityOverride=true
    simulationSpace=RootRelative
    alphaBoolBlend {blendOption=Linear}
    boundBone {boneName=dyn_Hair_l_01}
    componentAppliedLinearAccClamp=<2500 2500 2500>
    componentLinearAccScale=<0.3 0.3 0.35>
    componentLinearVelScale=<0.35 0.35 0.35>
    gravityOverride {z=-600}
    physicsBodyDefinitions [{
      sphereCollisionRadius=0
      boundBone {boneName=dyn_Hair_l_01}
      boxExtents=<5 10 10>
      constraintSetup {angularTargetAxis=AxisY, angularLimitsMax=<5 5 0>, angularLimitsMin=<-2 -2 -25>, angularTarget=<0 0 0>}
      localJointOffset=<-5 -0.1 -1>
     }]
    relativeSpaceBone {boneName=spine_05}
   }
   n55 = Anim Dynamics - Bone: dyn_Hair_l_02(ComponentPose=@n54, Alpha=@n53, ExternalForce="0,0,0") {
    angularDampingOverride=0.999
    bAngularSpring=true
    bGravityOverrideInSimSpace=true
    bOverrideAngularDamping=true
    bUseGravityOverride=true
    simulationSpace=RootRelative
    alphaBoolBlend {blendOption=Linear}
    boundBone {boneName=dyn_Hair_l_02}
    componentAppliedLinearAccClamp=<2500 2500 2500>
    componentLinearAccScale=<0.3 0.3 0.35>
    componentLinearVelScale=<0.35 0.35 0.35>
    gravityOverride {z=-600}
    physicsBodyDefinitions [{
      sphereCollisionRadius=0
      boundBone {boneName=dyn_Hair_l_02}
      boxExtents=<5 10 10>
      constraintSetup {angularTargetAxis=AxisY, angularLimitsMax=<5 5 0>, angularLimitsMin=<-2 -2 -25>, angularTarget=<0 0 0>}
      localJointOffset=<-5 -0.1 1>
     }]
    relativeSpaceBone {boneName=spine_05}
   }
   n56 = Anim Dynamics - Bone: dyn_Hair_l_03(ComponentPose=@n55, Alpha=@n53, ExternalForce="0,0,0") {
    angularDampingOverride=0.999
    bAngularSpring=true
    bGravityOverrideInSimSpace=true
    bOverrideAngularDamping=true
    bUseGravityOverride=true
    simulationSpace=RootRelative
    alphaBoolBlend {blendOption=Linear}
    boundBone {boneName=dyn_Hair_l_03}
    componentAppliedLinearAccClamp=<2500 2500 2500>
    componentLinearAccScale=<0.3 0.3 0.35>
    componentLinearVelScale=<0.35 0.35 0.35>
    gravityOverride {z=-600}
    physicsBodyDefinitions [{
      sphereCollisionRadius=0
      boundBone {boneName=dyn_Hair_l_03}
      boxExtents=<5 10 10>
      constraintSetup {angularTargetAxis=AxisY, angularLimitsMax=<5 5 0>, angularLimitsMin=<-2 -2 -25>, angularTarget=<0 0 0>}
      localJointOffset=<-5 -0.1 1>
     }]
    relativeSpaceBone {boneName=dyn_Hair_l_02}
   }
   n57 = Anim Dynamics - Bone: dyn_Hair_l_04(ComponentPose=@n56, Alpha=@n53, ExternalForce="0,0,0") {
    angularDampingOverride=0.999
    bAngularSpring=true
    bGravityOverrideInSimSpace=true
    bOverrideAngularDamping=true
    bUseGravityOverride=true
    simulationSpace=RootRelative
    alphaBoolBlend {blendOption=Linear}
    boundBone {boneName=dyn_Hair_l_04}
    componentAppliedLinearAccClamp=<2500 2500 2500>
    componentLinearAccScale=<0.3 0.3 0.35>
    componentLinearVelScale=<0.35 0.35 0.35>
    gravityOverride {z=-600}
    physicsBodyDefinitions [{
      sphereCollisionRadius=0
      boundBone {boneName=dyn_Hair_l_04}
      boxExtents=<5 10 10>
      constraintSetup {angularTargetAxis=AxisY, angularLimitsMax=<5 5 0>, angularLimitsMin=<-2 -2 -25>, angularTarget=<0 0 0>}
      localJointOffset=<-5 -1 1>
     }]
    relativeSpaceBone {boneName=spine_05}
   }
   n58 = Anim Dynamics - Bone: dyn_Hair_r_01(ComponentPose=@n57, Alpha=@n53, ExternalForce="0,0,0") {
    angularDampingOverride=0.999
    bAngularSpring=true
    bGravityOverrideInSimSpace=true
    bOverrideAngularDamping=true
    bUseGravityOverride=true
    simulationSpace=RootRelative
    alphaBoolBlend {blendOption=Linear}
    boundBone {boneName=dyn_Hair_r_01}
    componentAppliedLinearAccClamp=<2500 2500 2500>
    componentLinearAccScale=<0.3 0.3 0.35>
    componentLinearVelScale=<0.35 0.35 0.35>
    gravityOverride {z=-600}
    physicsBodyDefinitions [{
      sphereCollisionRadius=0
      boundBone {boneName=dyn_Hair_r_01}
      boxExtents=<5 10 10>
      constraintSetup {angularTargetAxis=AxisY, angularLimitsMax=<5 5 5>, angularLimitsMin=<-2 -5 -5>, angularTarget=<0 0 0>}
      localJointOffset=<5 -0.1 -1>
     }]
    relativeSpaceBone {boneName=spine_05}
   }
   n59 = Anim Dynamics - Bone: dyn_Hair_r_02(ComponentPose=@n58, Alpha=@n53, ExternalForce="0,0,0") {
    angularDampingOverride=0.999
    bAngularSpring=true
    bGravityOverrideInSimSpace=true
    bOverrideAngularDamping=true
    bUseGravityOverride=true
    simulationSpace=RootRelative
    alphaBoolBlend {blendOption=Linear}
    boundBone {boneName=dyn_Hair_r_02}
    componentAppliedLinearAccClamp=<2500 2500 2500>
    componentLinearAccScale=<0.3 0.3 0.35>
    componentLinearVelScale=<0.35 0.35 0.35>
    gravityOverride {z=-600}
    physicsBodyDefinitions [{
      sphereCollisionRadius=0
      boundBone {boneName=dyn_Hair_r_02}
      boxExtents=<5 10 10>
      constraintSetup {angularTargetAxis=AxisY, angularLimitsMax=<5 5 5>, angularLimitsMin=<-2 -5 -5>, angularTarget=<0 0 0>}
      localJointOffset=<5 -0.1 1>
     }]
    relativeSpaceBone {boneName=spine_05}
   }
   n60 = Anim Dynamics - Bone: dyn_Hair_r_03(ComponentPose=@n59, Alpha=@n53, ExternalForce="0,0,0") {
    angularDampingOverride=0.999
    bAngularSpring=true
    bGravityOverrideInSimSpace=true
    bOverrideAngularDamping=true
    bUseGravityOverride=true
    simulationSpace=RootRelative
    alphaBoolBlend {blendOption=Linear}
    boundBone {boneName=dyn_Hair_r_03}
    componentAppliedLinearAccClamp=<2500 2500 2500>
    componentLinearAccScale=<0.3 0.3 0.35>
    componentLinearVelScale=<0.35 0.35 0.35>
    gravityOverride {z=-600}
    physicsBodyDefinitions [{
      sphereCollisionRadius=0
      boundBone {boneName=dyn_Hair_r_03}
      boxExtents=<5 10 10>
      constraintSetup {angularTargetAxis=AxisY, angularLimitsMax=<5 5 5>, angularLimitsMin=<-2 -5 -5>, angularTarget=<0 0 0>}
      localJointOffset=<5 -0.1 1>
     }]
    relativeSpaceBone {boneName=dyn_Hair_r_02}
   }
   n61 = Anim Dynamics - Bone: dyn_Hair_r_04(ComponentPose=@n60, Alpha=@n53, ExternalForce="0,0,0") {
    angularDampingOverride=0.999
    bAngularSpring=true
    bGravityOverrideInSimSpace=true
    bOverrideAngularDamping=true
    bUseGravityOverride=true
    simulationSpace=RootRelative
    alphaBoolBlend {blendOption=Linear}
    boundBone {boneName=dyn_Hair_r_04}
    componentAppliedLinearAccClamp=<2500 2500 2500>
    componentLinearAccScale=<0.3 0.3 0.35>
    componentLinearVelScale=<0.35 0.35 0.35>
    gravityOverride {z=-600}
    physicsBodyDefinitions [{
      sphereCollisionRadius=0
      boundBone {boneName=dyn_Hair_r_04}
      boxExtents=<5 10 10>
      constraintSetup {angularTargetAxis=AxisY, angularLimitsMax=<5 5 5>, angularLimitsMin=<-2 -5 -5>, angularTarget=<0 0 0>}
      localJointOffset=<5 -1 1>
     }]
    relativeSpaceBone {boneName=spine_05}
   }
   n62 = Get Body_Enabled_0_4() -> (Body_Enabled_0_4:double)
   n63 = Anim Dynamics - Bone: dyn_Hair_Back_01(ComponentPose=@n61, Alpha=@n62, ExternalForce="0,0,0") {
    angularDampingOverride=0.999
    bAngularSpring=true
    bGravityOverrideInSimSpace=true
    bOverrideAngularDamping=true
    bUseGravityOverride=true
    simulationSpace=RootRelative
    alphaBoolBlend {blendOption=Linear}
    boundBone {boneName=dyn_Hair_Back_01}
    componentAppliedLinearAccClamp=<2500 2500 2500>
    componentLinearAccScale=<0.3 0.3 0.35>
    componentLinearVelScale=<0.35 0.35 0.35>
    gravityOverride {z=-600}
    physicsBodyDefinitions [{
      sphereCollisionRadius=0
      boundBone {boneName=dyn_Hair_Back_01}
      boxExtents=<5 10 10>
      constraintSetup {angularTargetAxis=AxisY, angularLimitsMax=<1 1 1>, angularLimitsMin=<-1 -1 -1>, angularTarget=<0 0 0>}
      localJointOffset=<-5 -0.1 -0.5>
     }]
    relativeSpaceBone {boneName=head}
   }
   n64 = Anim Dynamics - Bone: dyn_Hair_Back_02(ComponentPose=@n63, Alpha=@n62, ExternalForce="0,0,0") {
    angularDampingOverride=0.999
    bAngularSpring=true
    bGravityOverrideInSimSpace=true
    bOverrideAngularDamping=true
    bUseGravityOverride=true
    simulationSpace=RootRelative
    alphaBoolBlend {blendOption=Linear}
    boundBone {boneName=dyn_Hair_Back_02}
    componentAppliedLinearAccClamp=<2500 2500 2500>
    componentLinearAccScale=<0.3 0.3 0.35>
    componentLinearVelScale=<0.35 0.35 0.35>
    gravityOverride {z=-600}
    physicsBodyDefinitions [{
      sphereCollisionRadius=0
      boundBone {boneName=dyn_Hair_Back_02}
      boxExtents=<10 5 10>
      constraintSetup {angularTargetAxis=AxisY, angularLimitsMax=<5 1 45>, angularLimitsMin=<-5 -1 -7>, angularTarget=<0 0 0>}
      localJointOffset=<-2 -4 -0.5>
     }]
    relativeSpaceBone {boneName=dyn_Hair_Back_01}
   }
   n65 = Anim Dynamics - Bone: dyn_Hair_BackL_03(ComponentPose=@n64, Alpha=@n62, ExternalForce="0,0,0") {
    angularDampingOverride=0.999
    bAngularSpring=true
    bGravityOverrideInSimSpace=true
    bOverrideAngularDamping=true
    bUseGravityOverride=true
    simulationSpace=RootRelative
    alphaBoolBlend {blendOption=Linear}
    boundBone {boneName=dyn_Hair_BackL_03}
    componentAppliedLinearAccClamp=<800 800 800>
    componentLinearAccScale=<0.3 0.3 0.35>
    componentLinearVelScale=<0.35 0.35 0.35>
    gravityOverride {z=-800}
    physicsBodyDefinitions [{
      sphereCollisionRadius=0
      boundBone {boneName=dyn_Hair_BackL_03}
      boxExtents=<10 5 10>
      constraintSetup {angularTargetAxis=AxisY, angularLimitsMax=<30 20 60>, angularLimitsMin=<-30 -20 -7>, angularTarget=<0 0 0>}
      localJointOffset=<-0.1 -5 1>
     }]
    relativeSpaceBone {boneName=dyn_Hair_Back_02}
   }
   n66 = Anim Dynamics - Bone: dyn_Hair_BackL_04(ComponentPose=@n65, Alpha=@n62, ExternalForce="0,0,0") {
    angularDampingOverride=0.999
    bAngularSpring=true
    bGravityOverrideInSimSpace=true
    bOverrideAngularDamping=true
    bUseGravityOverride=true
    simulationSpace=RootRelative
    alphaBoolBlend {blendOption=Linear}
    boundBone {boneName=dyn_Hair_BackL_04}
    componentAppliedLinearAccClamp=<800 800 800>
    componentLinearAccScale=<0.3 0.3 0.35>
    componentLinearVelScale=<0.35 0.35 0.35>
    gravityOverride {z=-800}
    physicsBodyDefinitions [{
      sphereCollisionRadius=0
      boundBone {boneName=dyn_Hair_BackL_04}
      boxExtents=<10 5 10>
      constraintSetup {angularTargetAxis=AxisY, angularLimitsMax=<30 20 60>, angularLimitsMin=<-30 -20 -7>, angularTarget=<0 0 0>}
      localJointOffset=<-0.1 -5 1>
     }]
    relativeSpaceBone {boneName=dyn_Hair_Back_03}
   }
   n67 = Anim Dynamics - Bone: dyn_Hair_Back_05(ComponentPose=@n66, Alpha=@n62, ExternalForce="0,0,0") {
    angularDampingOverride=0.999
    bAngularSpring=true
    bGravityOverrideInSimSpace=true
    bOverrideAngularDamping=true
    bUseGravityOverride=true
    simulationSpace=RootRelative
    alphaBoolBlend {blendOption=Linear}
    boundBone {boneName=dyn_Hair_Back_05}
    componentAppliedLinearAccClamp=<800 800 800>
    componentLinearAccScale=<0.3 0.3 0.35>
    componentLinearVelScale=<0.35 0.35 0.35>
    gravityOverride {z=-800}
    physicsBodyDefinitions [{
      sphereCollisionRadius=0
      boundBone {boneName=dyn_Hair_Back_05}
      boxExtents=<10 5 10>
      constraintSetup {angularTargetAxis=AxisY, angularLimitsMax=<30 20 60>, angularLimitsMin=<-30 -20 -7>, angularTarget=<0 0 0>}
      localJointOffset=<-0.1 -5 1>
     }]
    relativeSpaceBone {boneName=dyn_Hair_Back_04}
   }
   n68 = Anim Dynamics - Bone: dyn_Hair_Back_06(ComponentPose=@n67, Alpha=@n62, ExternalForce="0,0,0") {
    angularDampingOverride=0.999
    bAngularSpring=true
    bGravityOverrideInSimSpace=true
    bOverrideAngularDamping=true
    bUseGravityOverride=true
    simulationSpace=RootRelative
    alphaBoolBlend {blendOption=Linear}
    boundBone {boneName=dyn_Hair_Back_06}
    componentAppliedLinearAccClamp=<800 800 800>
    componentLinearAccScale=<0.3 0.3 0.35>
    componentLinearVelScale=<0.35 0.35 0.35>
    gravityOverride {z=-800}
    physicsBodyDefinitions [{
      sphereCollisionRadius=0
      boundBone {boneName=dyn_Hair_Back_06}
      boxExtents=<10 5 10>
      constraintSetup {angularTargetAxis=AxisY, angularLimitsMax=<30 20 60>, angularLimitsMin=<-30 -20 10>, angularTarget=<0 0 0>}
      localJointOffset=<-0.1 -5 1>
     }]
    relativeSpaceBone {boneName=dyn_Hair_Back_05}
   }
   n69 = Anim Dynamics - Bone: dyn_Hair_Back_07(ComponentPose=@n68, Alpha=@n62, ExternalForce="0,0,0") {
    angularDampingOverride=0.999
    bAngularSpring=true
    bGravityOverrideInSimSpace=true
    bOverrideAngularDamping=true
    bUseGravityOverride=true
    simulationSpace=RootRelative
    alphaBoolBlend {blendOption=Linear}
    boundBone {boneName=dyn_Hair_Back_07}
    componentAppliedLinearAccClamp=<800 800 800>
    componentLinearAccScale=<0.3 0.3 0.35>
    componentLinearVelScale=<0.35 0.35 0.35>
    gravityOverride {z=-800}
    physicsBodyDefinitions [{
      sphereCollisionRadius=0
      boundBone {boneName=dyn_Hair_Back_07}
      boxExtents=<10 5 10>
      constraintSetup {angularTargetAxis=AxisY, angularLimitsMax=<30 20 60>, angularLimitsMin=<-30 -20 -7>, angularTarget=<0 0 0>}
      localJointOffset=<-0.1 -5 1>
     }]
    relativeSpaceBone {boneName=dyn_Hair_Back_04}
   }
   n70 = Anim Dynamics - Bone: dyn_Hair_Back_08(ComponentPose=@n69, Alpha=@n62, ExternalForce="0,0,0") {
    angularDampingOverride=0.999
    bAngularSpring=true
    bGravityOverrideInSimSpace=true
    bOverrideAngularDamping=true
    bUseGravityOverride=true
    simulationSpace=RootRelative
    alphaBoolBlend {blendOption=Linear}
    boundBone {boneName=dyn_Hair_Back_08}
    componentAppliedLinearAccClamp=<800 800 800>
    componentLinearAccScale=<0.3 0.3 0.35>
    componentLinearVelScale=<0.35 0.35 0.35>
    gravityOverride {z=-800}
    physicsBodyDefinitions [{
      sphereCollisionRadius=0
      boundBone {boneName=dyn_Hair_Back_08}
      boxExtents=<10 5 10>
      constraintSetup {angularTargetAxis=AxisY, angularLimitsMax=<30 20 60>, angularLimitsMin=<-30 -20 10>, angularTarget=<0 0 0>}
      localJointOffset=<-0.1 -5 1>
     }]
    relativeSpaceBone {boneName=dyn_Hair_Back_05}
   }
   n71 = "Transform (Modify) Bone - Bone: dyn_Hair_Back_03"(ComponentPose=@n70, Alpha=Get neck_rot_dn() -> (neck_rot_dn:double), Translation="0,0,0", Rotation="0,0,0", Scale="1,1,1") {
    rotationMode=BMM_Additive
    rotationSpace=BCS_BoneSpace
    scaleMode=BMM_Additive
    scaleSpace=BCS_BoneSpace
    translationMode=BMM_Additive
    translationSpace=BCS_BoneSpace
    alphaBoolBlend {blendOption=Linear}
    boneToModify {boneName=dyn_Hair_Back_03}
   }
   n72 = Get isWearingPackBool()
   n73 = "Constraint - dyn_Hair_Back_03 (PP)"!(ComponentPose=@n71, bAlphaBoolEnabled:bool=@n72, ConstraintWeights_0=0.5, ConstraintWeights_1:float=0.5) -> (Pose) {
    alphaInputType=Bool
    alphaBoolBlend {blendOption=Linear}
    boneToModify {boneName=dyn_Hair_Back_03}
    constraintSetup [{targetBone {boneName=spine_05}}, {targetBone {boneName=dyn_Hair_l_02}}]
    constraintWeights [1,1]
   }
   n74 = "Constraint - dyn_Hair_Back_04 (PP)"!(ComponentPose=@n73, bAlphaBoolEnabled:bool=@n72, ConstraintWeights_0=0.8, ConstraintWeights_1:float=0.2) -> (Pose) {
    alphaInputType=Bool
    alphaBoolBlend {blendOption=Linear}
    boneToModify {boneName=dyn_Hair_Back_04}
    constraintSetup [{targetBone {boneName=spine_05}}, {targetBone {boneName=dyn_Hair_l_03}}]
    constraintWeights [1,1]
   }
   n75 = "Constraint - dyn_Hair_Back_05 (PP)"!(ComponentPose=@n74, bAlphaBoolEnabled:bool=@n72, ConstraintWeights_0=1, ConstraintWeights_1:float=0.1) -> (Pose) {
    alphaInputType=Bool
    alphaBoolBlend {blendOption=Linear}
    boneToModify {boneName=dyn_Hair_Back_05}
    constraintSetup [{targetBone {boneName=spine_05}}, {targetBone {boneName=dyn_Hair_l_03}}]
    constraintWeights [1,1]
   }
   n76 = "Constraint - dyn_Hair_Back_06 (PP)"!(ComponentPose=@n75, bAlphaBoolEnabled:bool=@n72, ConstraintWeights_0=1, ConstraintWeights_1:float=0) -> (Pose) {
    alphaInputType=Bool
    alphaBoolBlend {blendOption=Linear}
    boneToModify {boneName=dyn_Hair_Back_06}
    constraintSetup [{targetBone {boneName=spine_05}}, {targetBone {boneName=dyn_Hair_l_03}}]
    constraintWeights [1,1]
   }
   n77 = "Constraint - dyn_Hair_Back_07 (PP)"!(ComponentPose=@n76, bAlphaBoolEnabled:bool=@n72, ConstraintWeights_0=1, ConstraintWeights_1:float=0) -> (Pose) {
    alphaInputType=Bool
    alphaBoolBlend {blendOption=Linear}
    boneToModify {boneName=dyn_Hair_Back_07}
    constraintSetup [{targetBone {boneName=spine_05}}, {targetBone {boneName=dyn_Hair_l_03}}]
    constraintWeights [1,1]
   }
   n78 = "Constraint - dyn_Hair_Back_08 (PP)"!(ComponentPose=@n77, bAlphaBoolEnabled:bool=@n72, ConstraintWeights_0=1, ConstraintWeights_1:float=0) -> (Pose) {
    alphaInputType=Bool
    alphaBoolBlend {blendOption=Linear}
    boneToModify {boneName=dyn_Hair_Back_08}
    constraintSetup [{targetBone {boneName=spine_05}}, {targetBone {boneName=dyn_Hair_l_03}}]
    constraintWeights [1,1]
   }
   n79 = "Constraint - dyn_Hair_BackL_01 (PP)"!(ComponentPose=@n78, Alpha=1, ConstraintWeights_0=0.3, ConstraintWeights_1:float=0.7) -> (Pose) {
    alphaBoolBlend {blendOption=Linear}
    boneToModify {boneName=dyn_Hair_BackL_01}
    constraintSetup [{targetBone {boneName=spine_05}}, {targetBone {boneName=dyn_Hair_Back_01}}]
    constraintWeights [1,1]
   }
   n80 = "Constraint - dyn_Hair_BackL_02 (PP)"!(ComponentPose=@n79, Alpha=1, ConstraintWeights_0=0.3, ConstraintWeights_1:float=0.7) -> (Pose) {
    alphaBoolBlend {blendOption=Linear}
    boneToModify {boneName=dyn_Hair_BackL_02}
    constraintSetup [{targetBone {boneName=spine_05}}, {targetBone {boneName=dyn_Hair_Back_02}}]
    constraintWeights [1,1]
   }
   n81 = "Constraint - dyn_Hair_BackL_03 (PP)"!(ComponentPose=@n80, Alpha=1, ConstraintWeights_0=0.3, ConstraintWeights_1:float=0.7) -> (Pose) {
    alphaBoolBlend {blendOption=Linear}
    boneToModify {boneName=dyn_Hair_BackL_03}
    constraintSetup [{targetBone {boneName=spine_05}}, {targetBone {boneName=dyn_Hair_Back_03}}]
    constraintWeights [1,1]
   }
   n82 = "Constraint - dyn_Hair_BackL_04 (PP)"!(ComponentPose=@n81, Alpha=1, ConstraintWeights_0=0.3, ConstraintWeights_1:float=0.7) -> (Pose) {
    alphaBoolBlend {blendOption=Linear}
    boneToModify {boneName=dyn_Hair_BackL_04}
    constraintSetup [{targetBone {boneName=spine_05}}, {targetBone {boneName=dyn_Hair_Back_04}}]
    constraintWeights [1,1]
   }
   n83 = "Constraint - dyn_Hair_BackL_05 (PP)"(ComponentPose=@n82, Alpha=1, ConstraintWeights_0=0.3, ConstraintWeights_1=0.7) {
    alphaBoolBlend {blendOption=Linear}
    boneToModify {boneName=dyn_Hair_BackL_05}
    constraintSetup [{targetBone {boneName=spine_05}}, {targetBone {boneName=dyn_Hair_Back_05}}]
    constraintWeights [1,1]
   }
   n84 = "Constraint - dyn_Hair_BackL_06 (PP)"(ComponentPose=@n83, Alpha=1, ConstraintWeights_0=0.3, ConstraintWeights_1=0.7) {
    alphaBoolBlend {blendOption=Linear}
    boneToModify {boneName=dyn_Hair_BackL_06}
    constraintSetup [{targetBone {boneName=spine_05}}, {targetBone {boneName=dyn_Hair_Back_06}}]
    constraintWeights [1,1]
   }
   n85 = "Constraint - dyn_Hair_BackL_07 (PP)"(ComponentPose=@n84, Alpha=1, ConstraintWeights_0=0.3, ConstraintWeights_1=0.7) {
    alphaBoolBlend {blendOption=Linear}
    boneToModify {boneName=dyn_Hair_BackL_07}
    constraintSetup [{targetBone {boneName=spine_05}}, {targetBone {boneName=dyn_Hair_Back_07}}]
    constraintWeights [1,1]
   }
   n86 = "Constraint - dyn_Hair_BackL_08 (PP)"(ComponentPose=@n85, Alpha=1, ConstraintWeights_0=0.3, ConstraintWeights_1=0.7) {
    alphaBoolBlend {blendOption=Linear}
    boneToModify {boneName=dyn_Hair_BackL_08}
    constraintSetup [{targetBone {boneName=spine_05}}, {targetBone {boneName=dyn_Hair_Back_08}}]
    constraintWeights [1,1]
   }
   n87 = "Constraint - dyn_Hair_BackR_01 (PP)"!(ComponentPose=@n86, Alpha=1, ConstraintWeights_0=0.3, ConstraintWeights_1:float=0.7) -> (Pose) {
    alphaBoolBlend {blendOption=Linear}
    boneToModify {boneName=dyn_Hair_BackR_01}
    constraintSetup [{targetBone {boneName=spine_05}}, {targetBone {boneName=dyn_Hair_Back_01}}]
    constraintWeights [1,1]
   }
   n88 = "Constraint - dyn_Hair_BackR_02 (PP)"!(ComponentPose=@n87, Alpha=1, ConstraintWeights_0=0.3, ConstraintWeights_1:float=0.7) -> (Pose) {
    alphaBoolBlend {blendOption=Linear}
    boneToModify {boneName=dyn_Hair_BackR_02}
    constraintSetup [{targetBone {boneName=spine_05}}, {targetBone {boneName=dyn_Hair_Back_02}}]
    constraintWeights [1,1]
   }
   n89 = "Constraint - dyn_Hair_BackR_03 (PP)"!(ComponentPose=@n88, Alpha=1, ConstraintWeights_0=0.3, ConstraintWeights_1:float=0.7) -> (Pose) {
    alphaBoolBlend {blendOption=Linear}
    boneToModify {boneName=dyn_Hair_BackR_03}
    constraintSetup [{targetBone {boneName=spine_05}}, {targetBone {boneName=dyn_Hair_Back_03}}]
    constraintWeights [1,1]
   }
   n90 = "Constraint - dyn_Hair_BackR_04 (PP)"!(ComponentPose=@n89, Alpha=1, ConstraintWeights_0=0.3, ConstraintWeights_1:float=0.7) -> (Pose) {
    alphaBoolBlend {blendOption=Linear}
    boneToModify {boneName=dyn_Hair_BackR_04}
    constraintSetup [{targetBone {boneName=spine_05}}, {targetBone {boneName=dyn_Hair_Back_04}}]
    constraintWeights [1,1]
   }
   n91 = "Constraint - dyn_Hair_BackR_05 (PP)"(ComponentPose=@n90, Alpha=1, ConstraintWeights_0=0.3, ConstraintWeights_1=0.7) {
    alphaBoolBlend {blendOption=Linear}
    boneToModify {boneName=dyn_Hair_BackR_05}
    constraintSetup [{targetBone {boneName=spine_05}}, {targetBone {boneName=dyn_Hair_Back_05}}]
    constraintWeights [1,1]
   }
   n92 = "Constraint - dyn_Hair_BackR_06 (PP)"(ComponentPose=@n91, Alpha=1, ConstraintWeights_0=0.3, ConstraintWeights_1=0.7) {
    alphaBoolBlend {blendOption=Linear}
    boneToModify {boneName=dyn_Hair_BackR_06}
    constraintSetup [{targetBone {boneName=spine_05}}, {targetBone {boneName=dyn_Hair_Back_06}}]
    constraintWeights [1,1]
   }
   n93 = "Constraint - dyn_Hair_BackR_07 (PP)"(ComponentPose=@n92, Alpha=1, ConstraintWeights_0=0.3, ConstraintWeights_1=0.7) {
    alphaBoolBlend {blendOption=Linear}
    boneToModify {boneName=dyn_Hair_BackR_07}
    constraintSetup [{targetBone {boneName=spine_05}}, {targetBone {boneName=dyn_Hair_Back_07}}]
    constraintWeights [1,1]
   }
   n94 = "Constraint - dyn_Hair_BackR_08 (PP)"(ComponentPose=@n93, Alpha=1, ConstraintWeights_0=0.3, ConstraintWeights_1=0.7) {
    alphaBoolBlend {blendOption=Linear}
    boneToModify {boneName=dyn_Hair_BackR_08}
    constraintSetup [{targetBone {boneName=spine_05}}, {targetBone {boneName=dyn_Hair_Back_08}}]
    constraintWeights [1,1]
   }
   n95 = "Transform (Modify) Bone - Bone: dyn_Hair_r_02"(ComponentPose=@n94, Alpha=Get FrontEnd() -> (FrontEnd:double), Translation="0,0,0", Rotation="10,-15,0", Scale="1,1,1") {
    rotationMode=BMM_Additive
    rotationSpace=BCS_BoneSpace
    scaleMode=BMM_Additive
    scaleSpace=BCS_BoneSpace
    translationMode=BMM_Additive
    translationSpace=BCS_BoneSpace
    alphaBoolBlend {blendOption=Linear}
    boneToModify {boneName=dyn_Hair_r_02}
    rotation {pitch=10, yaw=-15}
   }
   n96 = "Transform (Modify) Bone - Bone: backpack_collision"(ComponentPose=@n95, Alpha=1, Translation="0,-100,0", Rotation="0,0,0", Scale="1,1,1") {
    rotationMode=BMM_Additive
    rotationSpace=BCS_BoneSpace
    scaleMode=BMM_Additive
    scaleSpace=BCS_BoneSpace
    translationMode=BMM_Additive
    translationSpace=BCS_BoneSpace
    alphaBoolBlend {blendOption=Linear}
    boneToModify {boneName=backpack_collision}
    translation {y=-100}
   }
   Save cached pose 'BackpackCache'(Pose=Component To Local(ComponentPose=@n96)) {cacheName=BackpackCache}
   SelectVector(A="0, 0,-500", B="0,0,-880", bPickA=@n10)
   OR Boolean!(A=Get bIsSkydiveFloating(), B=Get bIsSkydiveDiveMode(), C:bool=Get bIsSkyDiving()) -> (ReturnValue)
  ]}
 EventGraph {nodes [
   Event BlueprintUpdateAnimation()
   > n1 = Sequence!(execute) -> (then_0, then_1, then_2:Exec, then_3:Exec)
   n2 = Get Skel_Mesh()
   > Set neck_rot_up(neck_rot_up:double=GetAlphaFromReferenceTransformDelta(SkeletalMeshComponent=@n2, BoneName=head, BaseName=spine_05, InputRange=(X=0,Y=-140), TwistAxis="0,0,1")) -> (Output_Get:double)
   > Set neck_rot_dn(neck_rot_dn:double=GetAlphaFromReferenceTransformDelta(SkeletalMeshComponent=@n2, BoneName=neck_01, BaseName=spine_05, DeltaComponentType=EulerX, InputRange=(X=-15,Y=50), TwistAxis="0,0,1")) -> (Output_Get:double)
   n3 = Get Shoulder_Mesh() -> (Shoulder_Mesh:Skeletal Mesh Component Object Reference)
   > Set Shoulder_rot_f_r(Shoulder_rot_f_r:double=GetAlphaFromReferenceTransformDelta(SkeletalMeshComponent=@n3, BoneName=upperarm_r, BaseName=clavicle_r, InputRange=(X=0,Y=90), TwistAxis="0,0,1")) -> (Output_Get:double)
   > Set Shoulder_rot_f_l(Shoulder_rot_f_l:double=GetAlphaFromReferenceTransformDelta(SkeletalMeshComponent=@n3, BoneName=upperarm_l, BaseName=clavicle_l, InputRange=(X=0,Y=90), TwistAxis="0,0,1")) -> (Output_Get:double)
   > Set Shoulder_rot_up_r(Shoulder_rot_up_r:double=GetAlphaFromReferenceTransformDelta(SkeletalMeshComponent=@n3, BoneName=clavicle_r, BaseName=spine_05, InputRange=(X=0,Y=110), TwistAxis="0,0,1")) -> (Output_Get:double)
   > Set Body_Enabled_0_4(Body_Enabled_0_4:double=SelectFloat(A=0, B=0.4, bPickA=Get bRigidBodyEnabled())) -> (Output_Get:double)
   > Set FrontEnd(FrontEnd:double="To Float (Boolean)"(InBool=Get bIsInFrontEnd())) -> (Output_Get:double)
   > Set Body_Enabled_0_5(Body_Enabled_0_5:double=SelectFloat(A=0, B=0.5, bPickA=Get bRigidBodyEnabled())) -> (Output_Get:double)
   n4 = Branch(execute=@n1.then_1, Condition=OR Boolean(A=Get bIsSkydiveFloating(), B=Get bIsSkydiveDiveMode(), C=Get bIsSkyDiving()))
   > Set Diving(Diving:bool=true) -> (Output_Get:bool)
   Set Diving(execute=@n4.else, Diving=false)
   n5 = Gravity Override State Function DEPRECATED(
    execute=@n1.then_2
    CharacterPartAnimInstance=Self-Reference()
    GameTimeInSeconds=GetGameTimeInSeconds()
    GravityOverrideParams=Get Gravity Override Params()
   )
   > Set Gravity Override Cape(Gravity Override Cape:Vector=@n5) -> (Output_Get:Vector)
   Event OnCharacterCustomizationLoaded()
   n6 = CheckForBackbling()
   > Set isWearingPackBool(isWearingPackBool:bool=OR Boolean(A=@n6, B=@n6.bIsBackblingACape, C=@n6.bIsBackblingLong)) -> (Output_Get:bool)
   > Set is_WearingCape(is_WearingCape:bool=@n6.bIsBackblingLong) -> (Output_Get:bool)
   > Set FrontEndCape(FrontEndCape:bool=AND Boolean(A=@n6.bIsBackblingLong, B=Get bIsInFrontEnd())) -> (Output_Get:bool)
   Event BlueprintInitializeAnimation()
   > Set Skel_Mesh(Skel_Mesh=GetOwnerSkeletalMeshComponent())
   > n7 = Cast To PlayerPawn_Athena(Object=TryGetPawnOwner())
   > n8 = Set Pawn Owner(Pawn Owner:Fort Player Pawn Athena Object Reference=@n7) -> (Output_Get:Fort Player Pawn Athena Object Reference)
   > Set Shoulder_Mesh(Shoulder_Mesh:Skeletal Mesh Component Object Reference=GetSkeletalMeshForPartType(self=@n8)) -> (Output_Get:Skeletal Mesh Component Object Reference)
   AND Boolean(A=@n6, B=NOT Boolean(A=@n6.bIsBackblingLong))
  ]}
]
variables [
 Body_Enabled_0_4 {type=number}
 Body_Enabled_0_5 {type=number}
 Diving {type=boolean}
 FrontEnd {type=number}
 FrontEndCape {type=boolean}
 Gravity Override Cape {type=Vector}
 Gravity Override Params {
  type=GravityOverrideParamsStruct
  default {
   onGround_Standing {windAmplitude=<-10 -150 0>, windFrequency=<-7 -1 -1>, windOffset=<-10 -100 0>}
   skydive_DivingDown {gravityOverride=<-500 0 0>}
   skydive_FloatingDown {gravityOverride=<0 -400 100>}
   skydive_GoingUp {gravityOverride=<0 -300 -200>}
   skydive_Parachuting {gravityOverride=<0 -600 -980>}
  }
 }
 is_WearingCape {type=boolean}
 isWearingPackBool {type=boolean}
 neck_rot_dn {type=number}
 neck_rot_up {type=number}
 Pawn Owner {type=FortPlayerPawnAthena}
 Shoulder_Mesh {type=SkeletalMeshComponent}
 Shoulder_rot_f_l {type=number}
 Shoulder_rot_f_r {type=number}
 Shoulder_rot_up_r {type=number}
 Skel_Mesh {type=SkeletalMeshComponent}
]
